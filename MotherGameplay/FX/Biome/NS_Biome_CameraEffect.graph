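# Niagara system for a biome camera effect, as its name indicates, built on one short-lived mesh particle with no user parameters.
# FN_Mesh bursts that mesh, shrinks it along X and fades alpha in and out.
# An active ScaleColor shifts it from purple through blue to cyan under an HDR intensity curve; two alternative ScaleColor modules are disabled.
# FN_Sprite, a GPU emitter scattering faint bluish sprites in a flat box, is disabled.
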NiagaraSystem /MotherGameplay/FX/Biome/NS_Biome_CameraEffect
dependencies {
 usedDataInterfaces [&NiagaraDataInterfaceCurve, &NiagaraDataInterfaceColorCurve]
 usedDynamicInputs []
 usedModules [&SystemState]
}
emitters {
 FN_Mesh {
  emitterUpdateScript [
   &EmitterState()
   &SpawnBurst_Instantaneous(Spawn Count=1, Spawn Time=0.02, Use Spawn Probability=false)
  ]
  particleSpawnScript [
   &InitializeParticle(Lifetime=0.75, Mesh Scale=<2 1.5 1.5>, Mesh Scale Mode=::NewEnumerator4, UsePositionOffset=false, Write Material Random=true)
   &InitialMeshOrientation(Rotation=<0 0 0.5>, Use Orientation Vector=true, Use Rotation Vector=true)
  ]
  particleUpdateScript [
   &ParticleState()
   &ScaleColor(
    Scale Alpha=&FloatFromCurve(FloatCurve=DataInterface() {propertyValues {Curve {keys [<at 0 1>,<at 1 0>]}}})
    Scale RGB=<5 5 5>
    ScaleA=true
    ScaleRGB=true
   ) {enabled=false}
   &ScaleColor(
    Scale Alpha=&FloatFromCurve(
     FloatCurve=DataInterface() {propertyValues {Curve {keys [
         {arriveTangent=2.5779, interpMode=RCIM_Cubic, leaveTangent=2.5779, tangentMode=RCTM_User, time=0, value=0}
         {arriveTangent=0.3312, interpMode=RCIM_Cubic, leaveTangent=-0.2993, tangentMode=RCTM_Break, time=0.2, value=1}
         {arriveTangent=-0.6165, interpMode=RCIM_Cubic, leaveTangent=-0.6165, tangentMode=RCTM_User, time=1, value=-0.05}
        ]}}}
     Scale Curve=&RandomRangeFloat(Minimum=1, Maximum=1)
    )
    Scale RGB=&Multiply_Vector(
     A=&MakeVectorFromLinearColorRGB(Color=&ColorFromCurve(ColorCurve=DataInterface() {propertyValues {
         AlphaCurve {keys [<at 0 1 RCIM_Cubic>]}
         BlueCurve {keys [
           <at 0.059 0.0984>
           {arriveTangent=2.2846, interpMode=RCIM_Cubic, leaveTangent=2.2846, time=0.3237, value=0}
           <at 0.5985 1.3309>
           <at 0.8027 5>
           <at 1 1 RCIM_Cubic>
          ]}
         GreenCurve {keys [
           <at 0.059 0>
           {arriveTangent=9.2682, interpMode=RCIM_Cubic, leaveTangent=9.2682, time=0.3237, value=0.6073}
           <at 0.5985 5>
           <at 0.8027 1.1479>
           <at 1 0.67 RCIM_Cubic>
          ]}
         RedCurve {keys [
           <at 0.059 1>
           {arriveTangent=-1.8536, interpMode=RCIM_Cubic, leaveTangent=-1.8536, time=0.3237, value=1}
           <at 0.5985 0>
           <at 0.8027 0.0984>
           <at 1 0.2546 RCIM_Cubic>
          ]}
        }}))
     B=&VectorFromFloat(Value=&FloatFromCurve(
       FloatCurve=DataInterface() {propertyValues {Curve {keys [
           {arriveTangent=9.0725, interpMode=RCIM_Cubic, leaveTangent=9.0725, tangentMode=RCTM_User, time=0, value=0.6}
           {arriveTangent=-2.5249, interpMode=RCIM_Cubic, leaveTangent=-2.5249, tangentMode=RCTM_User, time=0.1408, value=0.9847}
           {arriveTangent=-0.0405, interpMode=RCIM_Cubic, leaveTangent=-0.0405, tangentMode=RCTM_User, time=1, value=0}
          ]}}}
       Scale Curve=5
      ))
    )
    ScaleA=true
    ScaleRGB=true
   ) {enabled=false, module=ScaleColor001}
   &ScaleColor(
    Scale Alpha=&FloatFromCurve(
     FloatCurve=DataInterface() {propertyValues {Curve {keys [
         {arriveTangent=2.5779, interpMode=RCIM_Cubic, leaveTangent=2.5779, tangentMode=RCTM_User, time=0, value=0}
         {arriveTangent=0.3312, interpMode=RCIM_Cubic, leaveTangent=-0.2993, tangentMode=RCTM_Break, time=0.2, value=1}
         {arriveTangent=-0.6165, interpMode=RCIM_Cubic, leaveTangent=-0.6165, tangentMode=RCTM_User, time=1, value=-0.05}
        ]}}}
     Scale Curve=&RandomRangeFloat(Minimum=1, Maximum=1)
    )
    Scale RGB=&Multiply_Vector(
     A=&MakeVectorFromLinearColorRGB(Color=&ColorFromCurve(ColorCurve=DataInterface() {propertyValues {
         AlphaCurve {keys [<at 0 1 RCIM_Cubic>]}
         BlueCurve {keys [
           <at 0.059 1>
           {arriveTangent=7.4145, interpMode=RCIM_Cubic, leaveTangent=7.4145, time=0.3237, value=1}
           <at 0.5985 5>
           <at 0.8027 5>
           <at 1 1 RCIM_Cubic>
          ]}
         GreenCurve {keys [
           <at 0.059 0.01>
           {arriveTangent=-0.0186, interpMode=RCIM_Cubic, leaveTangent=-0.0186, time=0.3237, value=0.0277}
           <at 0.5985 0>
           <at 0.8027 2.6825>
           <at 1 0.67 RCIM_Cubic>
          ]}
         RedCurve {keys [
           <at 0.059 0.6448>
           {arriveTangent=-0.9327, interpMode=RCIM_Cubic, leaveTangent=-0.9327, time=0.3237, value=0.5935}
           <at 0.5985 0.1416>
           <at 0.8027 0>
           <at 1 0.2546 RCIM_Cubic>
          ]}
        }}))
     B=&VectorFromFloat(Value=&FloatFromCurve(
       FloatCurve=DataInterface() {propertyValues {Curve {keys [
           {arriveTangent=9.0725, interpMode=RCIM_Cubic, leaveTangent=9.0725, tangentMode=RCTM_User, time=0, value=0.6}
           {arriveTangent=-2.5249, interpMode=RCIM_Cubic, leaveTangent=-2.5249, tangentMode=RCTM_User, time=0.1408, value=0.9847}
           {arriveTangent=-0.0405, interpMode=RCIM_Cubic, leaveTangent=-0.0405, tangentMode=RCTM_User, time=1, value=0}
          ]}}}
       Scale Curve=10
      ))
    )
    ScaleA=true
    ScaleRGB=true
   ) {module=ScaleColor002}
   &ScaleMeshSize(Scale Factor=&MakeVector(
     X=&FloatFromCurve(FloatCurve=DataInterface() {propertyValues {Curve {keys [<at 0 1>,<at 1 0.5>]}}})
     Y=1
     Z=1
    ))
  ]
  rendererClasses [&NiagaraMeshRendererProperties]
  renderers [{rendererClass=&NiagaraMeshRendererProperties, rendererIndex=0}]
 }
 FN_Sprite {
  bEnabled=false
  simTarget=GPUComputeSim
  emitterUpdateScript [
   &EmitterState()
   &SpawnBurst_Instantaneous(Spawn Count=100, Spawn Time=0.02, Use Spawn Probability=false)
  ]
  particleSpawnScript [
   &InitializeParticle(
    Sprite Size Mode=::NewEnumerator1
    Uniform Sprite Size Max=0.3
    Uniform Sprite Size Min=0.3
    UsePositionOffset=false
    Write Material Random=true
   )
   &BoxLocation(Box Size=<0 17 11>, MaskSpawnGroup=false, Module.Override Seed=false, Rotate Box=false, Surface Only=false)
  ]
  particleUpdateScript [
   &ParticleState()
   &SolveForcesAndVelocity(Clamp Velocity=false, Limit Acceleration=false)
   &ScaleColor(Scale Alpha=0.2, Scale RGB=<0.4 0.6 1>, ScaleA=true, ScaleRGB=true)
   &DynamicMaterialParameters(
    Index 0 Param 2=15
    Index 0 Param 3=1
    Index 0 Param 4=1
    Module.Param0WriteEnabled=true
    Module.Param1WriteEnabled=true
    Module.Param2WriteEnabled=true
    Module.Param3WriteEnabled=true
   )
  ]
  rendererClasses [&NiagaraSpriteRendererProperties]
  renderers [{rendererClass=&NiagaraSpriteRendererProperties}]
 }
}
summary {
 systemName=NS_Biome_CameraEffect
 emitters [
  {emitterName=FN_Mesh, rendererClasses [&NiagaraMeshRendererProperties]}
  {bEnabled=false, emitterName=FN_Sprite, simTarget=GPUComputeSim, rendererClasses [&NiagaraSpriteRendererProperties]}
 ]
 userVariables []
}
