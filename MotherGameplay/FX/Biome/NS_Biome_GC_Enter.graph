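# Niagara effect for entering a biome, by its name, that surrounds a skeletal mesh with color-shifting flames and sparkles.
# GPU Flame sprites spawn at user skeletal mesh locations over a short emitter window, drifting with curl noise and flipbook animation.
# Two GPU sparkle emitters snap sphere-spawned particles onto the depth buffer, sample GBuffer color, kill invalid or distant hits, and inherit source movement.
# An attached skeletal-mesh ribbon emitter is present but disabled.
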NiagaraSystem /MotherGameplay/FX/Biome/NS_Biome_GC_Enter
dependencies {
 usedDataInterfaces [&NiagaraDataInterfaceGBuffer, &NiagaraDataInterfaceCurve, &NiagaraDataInterfaceColorCurve]
 usedDynamicInputs []
 usedModules [&SystemState]
}
emitters {
 FN_AttachedRibbon {
  bEnabled=false
  emitterUpdateScript [
   &EmitterState()
   &FN_SpawnRibbon_v2(Number of Ribbons=4, Spawn Mode=::NewEnumerator0, SpawnRate=30)
   &FN_Integrator() {enabled=false}
  ]
  particleSpawnScript [
   /Game/Effects/Niagara/Modules/Ribbons/FN_InitializeRibbon_v4(
    Color:LinearColor=<rgba 150 150 150 1>
    Enable Skeletal Position=true
    Ribbon Width=25
    SkeletalMesh:NiagaraDataInterfaceSkeletalMesh=$User.SkeletalMesh
   )
   &FN_FlipRibbonV() {enabled=false}
   &AddVelocity(Velocity=<0 0 100>)
  ]
  particleUpdateScript [
   &SolveForcesAndVelocity(Clamp Velocity=false, Limit Acceleration=false)
   /Game/Effects/Niagara/Modules/Ribbons/FN_UpdateRibbon_v4(Enable Skeletal Position=true, SkeletalMesh:NiagaraDataInterfaceSkeletalMesh=$User.SkeletalMesh)
   &FN_RibbonMaterial(
    Color=&MakeLinearColorFromVectorAndFloat(
     "Vector (RGB)"=&Multiply_Vector(
      A=&MakeVectorFromLinearColorRGB(Color=&ColorFromCurve(ColorCurve=DataInterface() {propertyValues {
          AlphaCurve {keys [<at 0 1 RCIM_Cubic>,<at 0.9977 1 RCIM_Cubic>]}
          BlueCurve {keys [<at 0.1175 0.3372 RCIM_Cubic>,<at 0.2107 0.1377>,<at 0.2992 0.7758>,<at 0.435 0.2206>,<at 0.5572 1.5>,<at 0.9071 0.807 RCIM_Cubic>]}
          GreenCurve {keys [<at 0.1175 0.0301 RCIM_Cubic>,<at 0.2107 0.6131>,<at 0.2992 0.2906>,<at 0.435 0.7084>,<at 0.5572 0.4532>,<at 0.9071 0.481 RCIM_Cubic>]}
          RedCurve {keys [<at 0.1175 0 RCIM_Cubic>,<at 0.2107 0.7305>,<at 0.2992 0.107>,<at 0.435 0.1782>,<at 0.5572 0>,<at 0.9071 0.1113 RCIM_Cubic>]}
         }}))
      B=&VectorFromFloat(Value=&FloatFromCurve(
        FloatCurve=DataInterface() {propertyValues {Curve {keys [
            {interpMode=RCIM_Cubic, tangentMode=RCTM_User, time=0, value=0}
            {interpMode=RCIM_Cubic, tangentMode=RCTM_User, time=0.5, value=1}
            {tangentMode=RCTM_User, time=1, value=0}
           ]}}}
        Scale Curve=0.25
       ))
     )
     "Float (Alpha)"=1
    )
    Erosion Amount=0.25
    Erosion Pan Speed U=1
    Macro Distortion Amount=1
    Macro Pan Offset=1
    Macro Pan Speed=1
    Micro Distortion Amount=1
    Opacity Erosion:NiagaraFloat=&FloatFromCurve(FloatCurve=DataInterface() {propertyValues {Curve {keys [<at 0 0>,<at 1 0.5>]}}})
    Opacity Pan U Offset=1
    Opacity Pan U Speed=0.25
    Opacity Scale U=0.5
    Scale Alpha:NiagaraFloat=&FloatFromCurve(
     FloatCurve=DataInterface() {propertyValues {Curve {keys [
         {interpMode=RCIM_Cubic, tangentMode=RCTM_User, time=0, value=0}
         {arriveTangent=-2.8813, interpMode=RCIM_Cubic, leaveTangent=-2.8813, tangentMode=RCTM_User, time=0.1, value=0.95}
         {interpMode=RCIM_Cubic, tangentMode=RCTM_User, time=0.4, value=0.25}
         {interpMode=RCIM_Cubic, tangentMode=RCTM_User, time=1, value=0}
        ]}}}
     Scale Curve=2
    )
   ) {module=FN_RibbonMaterial001}
  ]
  rendererClasses [&NiagaraRibbonRendererProperties]
  renderers [{rendererClass=&NiagaraRibbonRendererProperties}]
 }
 Flame {
  simTarget=GPUComputeSim
  emitterUpdateScript [
   &EmitterState(Life Cycle Mode=::NewEnumerator1, Loop Behavior=::NewEnumerator1, Loop Duration=0.25, UseLoopDelay=false)
   &SpawnRate(SpawnRate=10, Use Spawn Probability=false)
  ]
  particleSpawnScript [
   &InitializeParticle(
    Color=&RandomRangeLinearColor(Minimum=<rgba 1 1 1 1>, Maximum=<rgba 1 1 1 1>)
    Lifetime Min=2
    Lifetime Mode=::NewEnumerator1
    Sprite Rotation Mode=::NewEnumerator2
    Sprite Size Mode=::NewEnumerator1
    Uniform Sprite Size Max=60
    Uniform Sprite Size Min=50
    UsePositionOffset=false
    Write Material Random=true
   ) {module=InitializeParticle001}
   /SetVariables(Particles.RandomFrame=&RandomRangeInt(Minimum=0, Maximum=63)) {isSetParameters=true}
   &DynamicMaterialParameters(
    Index 0 Param 1=&RandomRangeFloat(Minimum=0.2, Maximum=0.5)
    Module.Param0WriteEnabled=true
    Module.Param1WriteEnabled=true
    Module.Param2WriteEnabled=true
    Module.Param3WriteEnabled=true
   )
   /Game/Effects/Niagara/Modules/Spawn/FN_Spawn_GBuffer_Color() {enabled=false}
   &SphereLocation(
    MaskSpawnGroup=false
    Override Seed=false
    Sphere Radius=25
    Surface Expansion Mode=::NewEnumerator2
    Surface Only=true
    Surface Only Band Thickness=1
    UseNonUniformScale=false
    UseOffset=false
    UseSphereOrientation=false
   )
   &SkeletalMeshLocation(
    Bone / Socket Index=&SpawnGroup()
    Bone Sampling Mode=::NewEnumerator4
    Inherit Velocity=false
    MaskSpawnGroup=false
    Sampled Position Offset=&RandomRangeVector(Minimum=<7 7 5>, Maximum=<-7 -7 -5>)
    Skeletal Mesh=$User.SkeletalMesh
   )
   &AddVelocityFromPoint(Module.Use Falloff Distance=false, Velocity Strength=30)
   &CameraOffset(Camera Offset Amount=&RandomRangeFloat(Minimum=0, Maximum=10))
  ]
  particleUpdateScript [
   &CurlNoiseForce(Noise Frequency=500, Noise Strength=500, Pan Noise=false)
   &Drag(Drag=&RandomRangeFloat(Minimum=2, Maximum=3), Use Linear Drag=true, Use Rotational Drag=false)
   &SolveForcesAndVelocity(Clamp Velocity=false, Limit Acceleration=false)
   &ParticleState()
   &FN_SubUVAnimation(End Frame=&Add_Integer(A=$Particles.RandomFrame, B=63), Start Frame=$Particles.RandomFrame)
   &ScaleSpriteSize(Scale Factor=&Vector2DFromFloat(Value=&FloatFromCurve(FloatCurve=DataInterface() {propertyValues {Curve {keys [
          {arriveTangent=2.5076, interpMode=RCIM_Cubic, leaveTangent=2.5076, tangentMode=RCTM_User, time=0.0023, value=0}
          {arriveTangent=0.0897, interpMode=RCIM_Cubic, leaveTangent=0.0897, tangentMode=RCTM_User, time=1, value=1}
         ]}}})))
   &ScaleColor(
    Scale Alpha=&FloatFromCurve(FloatCurve=DataInterface() {propertyValues {Curve {keys [
         {interpMode=RCIM_Cubic, tangentMode=RCTM_User, time=0, value=0}
         {interpMode=RCIM_Cubic, tangentMode=RCTM_User, time=0.15, value=1}
         {interpMode=RCIM_Cubic, tangentMode=RCTM_User, time=0.8, value=1}
         {interpMode=RCIM_Cubic, tangentMode=RCTM_User, time=1, value=0}
        ]}}})
    Scale RGB=&Multiply_Vector(
     A=&MakeVectorFromLinearColorRGB(Color=&ColorFromCurve(ColorCurve=DataInterface() {propertyValues {
         AlphaCurve {keys [<at 0 1 RCIM_Cubic>,<at 0.9977 1 RCIM_Cubic>]}
         BlueCurve {keys [<at 0.1175 0.0035 RCIM_Cubic>,<at 0.2682 0.1377>,<at 0.4507 0.2122>,<at 0.5951 0.5271>,<at 0.7691 1.5>,<at 0.9071 0.521 RCIM_Cubic>]}
         GreenCurve {keys [<at 0.1175 0.0068 RCIM_Cubic>,<at 0.2682 0.6131>,<at 0.4507 0.2346>,<at 0.5951 0.4287>,<at 0.7691 0.4532>,<at 0.9071 0.4852 RCIM_Cubic>]}
         RedCurve {keys [<at 0.1175 0.3372 RCIM_Cubic>,<at 0.2682 0.7305>,<at 0.4507 0.7758>,<at 0.5951 0.7084>,<at 0.7691 0>,<at 0.9071 0.807 RCIM_Cubic>]}
        }}))
     B=&VectorFromFloat(Value=&FloatFromCurve(
       FloatCurve=DataInterface() {propertyValues {Curve {keys [
           {interpMode=RCIM_Cubic, tangentMode=RCTM_User, time=0, value=0}
           {interpMode=RCIM_Cubic, tangentMode=RCTM_User, time=0.5, value=1}
           {tangentMode=RCTM_User, time=1, value=0}
          ]}}}
       Scale Curve=0.25
      ))
    )
    ScaleA=true
    ScaleRGB=true
   )
  ]
  rendererClasses [&NiagaraSpriteRendererProperties]
  renderers [{rendererClass=&NiagaraSpriteRendererProperties}]
 }
 TestSparkles {
  simTarget=GPUComputeSim
  emitterUpdateScript [
   &EmitterState()
   &SpawnRate(SpawnRate=2500, Use Spawn Probability=false)
  ]
  particleSpawnScript [
   /SetVariables() {isSetParameters=true, module=SetVariables_#1}
   &InitializeParticle(
    AdjustAlpha=false
    AdjustHue=true
    AdjustSaturation=false
    AdjustValue=false
    Color=<rgba 15 15 15 1>
    Color Mode=::NewEnumerator3
    Lifetime Max=1
    Lifetime Mode=::NewEnumerator1
    Sprite Rotation Mode=::NewEnumerator2
    Sprite Size Mode=::NewEnumerator1
    Uniform Sprite Size Max=8
    Uniform Sprite Size Min=4
    UsePositionOffset=false
    Write Material Random=true
   ) {module=InitializeParticle001}
   &SphereLocation(
    MaskSpawnGroup=false
    Non Uniform Scale=<0.1 1 1>
    Offset=&MakeVector(X=0, Y=0, Z=0)
    Override Seed=false
    Sphere Radius=200
    Surface Only=false
    UseNonUniformScale=true
    UseOffset=true
    UseSphereOrientation=false
   )
   /SetVariables(Particles.InitialPos:NiagaraPosition=$Particles.Position) {isSetParameters=true, module=SetVariables_#2}
   /Niagara/Modules/Math/PlaceParticlesOnDepthBuffer_GPU(Max Z Depth=250, Module.Use Max Z Depth=true)
   &FN_Spawn_GBuffer_Color()
   &KillParticles(Kill Particles=&SetBoolByFloatComparison(
     A=&VectorLength(Vector=&Subtract_Vector(A=$Particles.InitialPos, B=Linked() {linkedVariable {name=Particles.Position, type=NiagaraPosition}}))
     B=1000
    ))
   &KillParticles(Kill Particles=&Boolean_NotOperation(Bool=$Output.PlaceParticlesOnDepthBuffer_GPU.ValidSample)) {module=KillParticles001}
   &AddVelocity(Velocity=<50 250 0>)
   &CameraOffset(Camera Offset Amount=25)
  ]
  particleUpdateScript [
   &ParticleState()
   &InheritSourceMovement(
    Applied Force Falloff Distance=25000
    Applied Force Scale=&VectorFromFloat(Value=&FloatFromCurve(FloatCurve=DataInterface() {propertyValues {Curve {keys [<at 0.2 1>,<at 0.25 0>]}}}, CurveIndex=$System.LoopedAge))
    Applied Force Speed Limit=2500
    Applied Position Falloff Distance=0
    Applied Position Scale=<-0.25 -0.25 -0.25>
    ApplyForce=true
    ApplyForceFalloff=false
    ApplyPosition=true
    ApplyPositionFalloff=false
    Module.LimitForceSpeed=true
   )
   &InheritVelocity(Module.Limit Speed=true, Velocity Limit=1)
   &CurlNoiseForce(Mask Curl Noise=true, Noise Strength=250, Pan Noise=false)
   &GravityForce(Gravity=<0 0 -5>)
   &Drag(Drag=2, Use Linear Drag=true, Use Rotational Drag=false)
   &SolveForcesAndVelocity(Clamp Velocity=false, Limit Acceleration=false)
   /SetVariables(Particles.GetAge:NiagaraFloat=$Particles.NormalizedAge) {isSetParameters=true, module=SetVariables_#3}
   &ScaleColor(
    Scale Alpha=&FloatFromCurve(FloatCurve=DataInterface() {propertyValues {Curve {keys [
         {interpMode=RCIM_Cubic, tangentMode=RCTM_User, time=0, value=0}
         {interpMode=RCIM_Cubic, tangentMode=RCTM_User, time=0.15, value=1}
         {interpMode=RCIM_Cubic, tangentMode=RCTM_User, time=0.8, value=1}
         {interpMode=RCIM_Cubic, tangentMode=RCTM_User, time=1, value=0}
        ]}}})
    Scale RGB=&Multiply_Vector(
     A=&MakeVectorFromLinearColorRGB(Color=&ColorFromCurve(ColorCurve=DataInterface() {propertyValues {
         AlphaCurve {keys [<at 0 1 RCIM_Cubic>,<at 0.9977 1 RCIM_Cubic>]}
         BlueCurve {keys [<at 0.1175 1 RCIM_Cubic>,<at 0.2682 1>,<at 0.4507 0.2122>,<at 0.5951 0.5271>,<at 0.7691 1.5>,<at 0.9071 0.521 RCIM_Cubic>]}
         GreenCurve {keys [<at 0.1175 1 RCIM_Cubic>,<at 0.2682 1>,<at 0.4507 0.2346>,<at 0.5951 0.4287>,<at 0.7691 0.4532>,<at 0.9071 0.4852 RCIM_Cubic>]}
         RedCurve {keys [<at 0.1175 1 RCIM_Cubic>,<at 0.2682 1>,<at 0.4507 0.7758>,<at 0.5951 0.7084>,<at 0.7691 0>,<at 0.9071 0.807 RCIM_Cubic>]}
        }}))
     B=&VectorFromFloat(Value=5)
    )
    ScaleA=true
    ScaleRGB=true
   )
   &ScaleSpriteSize(Scale Factor=&Vector2DFromFloat(Value=&FloatFromCurve(FloatCurve=DataInterface() {propertyValues {Curve {keys [
          {interpMode=RCIM_Cubic, tangentMode=RCTM_User, time=-0.0044, value=0.1773}
          {arriveTangent=-0.1766, interpMode=RCIM_Cubic, leaveTangent=-0.1766, tangentMode=RCTM_User, time=0.1174, value=0.9496}
          {arriveTangent=-1.076, interpMode=RCIM_Cubic, leaveTangent=-1.076, tangentMode=RCTM_User, time=0.1639, value=0.7081}
          <at 0.8264 0.7215>
          <at 1 0>
         ]}}})))
   &ScaleSpriteSizeBySpeed(Max Scale Factor=<0.6 2>, Min Scale Factor=<0.5 1>, Velocity Threshold=800) {enabled=false}
  ]
  rendererClasses [&NiagaraSpriteRendererProperties]
  renderers [{rendererClass=&NiagaraSpriteRendererProperties}]
 }
 TestSparkles001 {
  simTarget=GPUComputeSim
  emitterUpdateScript [
   &EmitterState(Life Cycle Mode=::NewEnumerator1, Loop Behavior=::NewEnumerator1, Loop Duration=0.25, UseLoopDelay=false)
   &SpawnRate(SpawnRate=1500, Use Spawn Probability=false)
  ]
  particleSpawnScript [
   /SetVariables() {isSetParameters=true, module=SetVariables_#1}
   &InitializeParticle(
    AdjustAlpha=false
    AdjustHue=true
    AdjustSaturation=false
    AdjustValue=false
    Color=<rgba 15 15 15 1>
    Color Mode=::NewEnumerator3
    Lifetime Max=1
    Lifetime Mode=::NewEnumerator1
    Sprite Rotation Mode=::NewEnumerator2
    Sprite Size Mode=::NewEnumerator1
    Uniform Sprite Size Max=15
    Uniform Sprite Size Min=15
    UsePositionOffset=false
    Write Material Random=true
   ) {module=InitializeParticle001}
   &SphereLocation(
    MaskSpawnGroup=false
    Non Uniform Scale=<0.1 1 1>
    Offset=&MakeVector(X=0, Y=0, Z=0)
    Override Seed=false
    Sphere Radius=200
    Surface Only=false
    UseNonUniformScale=true
    UseOffset=true
    UseSphereOrientation=false
   )
   /SetVariables(Particles.InitialPos:NiagaraPosition=$Particles.Position) {isSetParameters=true, module=SetVariables_#2}
   &PlaceParticlesOnDepthBuffer_GPU(Max Z Depth=250, Module.Use Max Z Depth=true)
   &FN_Spawn_GBuffer_Color()
   &KillParticles(Kill Particles=&SetBoolByFloatComparison(
     A=&VectorLength(Vector=&Subtract_Vector(A=$Particles.InitialPos, B=Linked() {linkedVariable {name=Particles.Position, type=NiagaraPosition}}))
     B=1000
    ))
   &KillParticles(Kill Particles=&Boolean_NotOperation(Bool=$Output.PlaceParticlesOnDepthBuffer_GPU.ValidSample)) {module=KillParticles001}
   &AddVelocity(Velocity=<50 250 0>)
   &CameraOffset(Camera Offset Amount=25)
   &SubUVAnimation(End Frame=3, SubUV Animation Mode:ENiagara_SubUVLookupMode=::NewEnumerator2)
  ]
  particleUpdateScript [
   &ParticleState()
   &InheritSourceMovement(
    Applied Force Falloff Distance=25000
    Applied Force Scale=&VectorFromFloat(Value=&FloatFromCurve(FloatCurve=DataInterface() {propertyValues {Curve {keys [<at 0.2 1>,<at 0.25 0>]}}}, CurveIndex=$System.LoopedAge, Scale Curve=0.1))
    Applied Force Speed Limit=2500
    Applied Position Falloff Distance=0
    Applied Position Scale=<-0.25 -0.25 -0.25>
    ApplyForce=true
    ApplyForceFalloff=false
    ApplyPosition=true
    ApplyPositionFalloff=false
    Module.LimitForceSpeed=true
   )
   &InheritVelocity(Module.Limit Speed=true, Velocity Limit=1, Velocity Scale=<0.25 0.25 0.25>)
   &CurlNoiseForce(Mask Curl Noise=true, Noise Frequency=1, Noise Strength=150, Pan Noise=false)
   &GravityForce(Gravity=<0 0 -5>)
   &Drag(Drag=2, Use Linear Drag=true, Use Rotational Drag=false)
   &SolveForcesAndVelocity(Clamp Velocity=false, Limit Acceleration=false)
   /SetVariables(Particles.GetAge:NiagaraFloat=$Particles.NormalizedAge) {isSetParameters=true, module=SetVariables_#3}
   &ScaleColor(
    Scale Alpha=&FloatFromCurve(FloatCurve=DataInterface() {propertyValues {Curve {keys [
         {interpMode=RCIM_Cubic, tangentMode=RCTM_User, time=0, value=0}
         {interpMode=RCIM_Cubic, tangentMode=RCTM_User, time=0.15, value=1}
         {interpMode=RCIM_Cubic, tangentMode=RCTM_User, time=0.8, value=1}
         {interpMode=RCIM_Cubic, tangentMode=RCTM_User, time=1, value=0}
        ]}}})
    Scale RGB=&Multiply_Vector(
     A=&MakeVectorFromLinearColorRGB(Color=&ColorFromCurve(ColorCurve=DataInterface() {propertyValues {
         AlphaCurve {keys [<at 0 1 RCIM_Cubic>,<at 0.9977 1 RCIM_Cubic>]}
         BlueCurve {keys [<at 0.1175 1 RCIM_Cubic>,<at 0.2682 1>,<at 0.4507 0.2122>,<at 0.5951 0.5271>,<at 0.7691 1.5>,<at 0.9071 0.521 RCIM_Cubic>]}
         GreenCurve {keys [<at 0.1175 1 RCIM_Cubic>,<at 0.2682 1>,<at 0.4507 0.2346>,<at 0.5951 0.4287>,<at 0.7691 0.4532>,<at 0.9071 0.4852 RCIM_Cubic>]}
         RedCurve {keys [<at 0.1175 1 RCIM_Cubic>,<at 0.2682 1>,<at 0.4507 0.7758>,<at 0.5951 0.7084>,<at 0.7691 0>,<at 0.9071 0.807 RCIM_Cubic>]}
        }}))
     B=&VectorFromFloat(Value=150)
    )
    ScaleA=true
    ScaleRGB=true
   )
   &ScaleSpriteSize(Scale Factor=&Vector2DFromFloat(Value=&FloatFromCurve(FloatCurve=DataInterface() {propertyValues {Curve {keys [
          {interpMode=RCIM_Cubic, tangentMode=RCTM_User, time=-0.0044, value=0.1773}
          {arriveTangent=-0.1766, interpMode=RCIM_Cubic, leaveTangent=-0.1766, tangentMode=RCTM_User, time=0.1174, value=0.9496}
          {arriveTangent=-1.076, interpMode=RCIM_Cubic, leaveTangent=-1.076, tangentMode=RCTM_User, time=0.1639, value=0.7081}
          <at 0.8264 0.7215>
          <at 1 0>
         ]}}})))
   &DynamicMaterialParameters(
    Index 0 Param 2=1
    Index 0 Param 3=1
    Index 0 Param 4=15
    Module.Param0WriteEnabled=true
    Module.Param1WriteEnabled=true
    Module.Param2WriteEnabled=true
    Module.Param3WriteEnabled=true
   )
   &ScaleSpriteSizeBySpeed(Max Scale Factor=<0.6 2>, Min Scale Factor=<0.5 1>, Velocity Threshold=800) {enabled=false}
  ]
  rendererClasses [&NiagaraSpriteRendererProperties]
  renderers [
   {rendererClass=&NiagaraSpriteRendererProperties}
   {rendererClass=&NiagaraSpriteRendererProperties, rendererIndex=1}
  ]
 }
}
summary {
 systemName=NS_Biome_GC_Enter
 emitters [
  {emitterName=TestSparkles, simTarget=GPUComputeSim, rendererClasses [&NiagaraSpriteRendererProperties]}
  {emitterName=Flame, simTarget=GPUComputeSim, rendererClasses [&NiagaraSpriteRendererProperties]}
  {bEnabled=false, emitterName=FN_AttachedRibbon, rendererClasses [&NiagaraRibbonRendererProperties]}
  {emitterName=TestSparkles001, simTarget=GPUComputeSim, rendererClasses [&NiagaraSpriteRendererProperties]}
 ]
 userVariables [
  User.bIsFrontEnd {type=NiagaraBool}
  User.bIsFrontEndPreview {type=NiagaraBool}
  User.FEVec {type=Vector3f, description="", defaultValue {struct=Vector3f, value=<0 0 0>}}
  User.RibbonIDOffset {type=NiagaraInt32}
  User.SkeletalMesh {
   type=NiagaraDataInterfaceSkeletalMesh
   defaultValue {
    struct=NiagaraExt_VariableValue_DataInterface
    value {dataInterface=&self:NiagaraDataInterfaceSkeletalMesh_0}
   }
  }
 ]
}
