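# Looping Niagara biome effect that trails colored, fading ribbons from a skeletal mesh supplied through User.SkeletalMesh.
# Two continuously spawning emitters share custom FN ribbon modules that position particles on the mesh.
# The first adds curl-noise drift and drives color gradients, erosion and panning distortion through FN_RibbonMaterial.
# The wider, denser second has zero-strength noise, disables FN_RibbonMaterial, colors via curve-driven ScaleColor, and lists sprite and ribbon renderers.
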NiagaraSystem /MotherGameplay/FX/Biome/NS_Biome_GC_Loop
dependencies {
 usedDataInterfaces [&NiagaraDataInterfaceColorCurve, &NiagaraDataInterfaceCurve]
 usedDynamicInputs []
 usedModules [&SystemState]
}
emitters {
 FN_AttachedRibbon001 {
  emitterUpdateScript [
   &EmitterState()
   &FN_SpawnRibbon_v2(Spawn Mode=::NewEnumerator0, SpawnRate=25)
   &FN_Integrator() {enabled=false}
  ]
  particleSpawnScript [
   /Game/Effects/Niagara/Modules/Ribbons/FN_InitializeRibbon_v4(Enable Skeletal Position=true, Lifetime=1, Ribbon Width=25, SkeletalMesh:NiagaraDataInterfaceSkeletalMesh=$User.SkeletalMesh)
   &FN_FlipRibbonV() {enabled=false}
   &AddVelocity()
   &CameraOffset()
  ]
  particleUpdateScript [
   &CurlNoiseForce(Noise Frequency=150, Noise Strength=200, Pan Noise=false)
   &SolveForcesAndVelocity(Clamp Velocity=false, Limit Acceleration=false)
   /Game/Effects/Niagara/Modules/Ribbons/FN_UpdateRibbon_v4(Enable Skeletal Position=true, SkeletalMesh:NiagaraDataInterfaceSkeletalMesh=$User.SkeletalMesh)
   &FN_RibbonMaterial(
    Color=&MakeLinearColorFromVectorAndFloat(
     "Vector (RGB)"=&Multiply_Vector(
      A=&MakeVectorFromLinearColorRGB(Color=&ColorFromCurve(ColorCurve=DataInterface() {propertyValues {
          AlphaCurve {keys [<at 0 1 RCIM_Cubic>,<at 0.9977 1 RCIM_Cubic>]}
          BlueCurve {keys [<at 0.1175 0.4118 RCIM_Cubic>,<at 0.2412 0.7758>,<at 0.4228 0.1176>,<at 0.606 0 RCIM_Cubic>]}
          GreenCurve {keys [<at 0.1175 1 RCIM_Cubic>,<at 0.2412 0.2906>,<at 0.4228 0.1412>,<at 0.606 0.5771 RCIM_Cubic>]}
          RedCurve {keys [<at 0.1175 0 RCIM_Cubic>,<at 0.2412 0.107>,<at 0.4228 2>,<at 0.606 0.807 RCIM_Cubic>]}
         }}))
      B=&VectorFromFloat(Value=&FloatFromCurve(
        FloatCurve=DataInterface() {propertyValues {Curve {keys [
            {interpMode=RCIM_Cubic, tangentMode=RCTM_User, time=0, value=0}
            {interpMode=RCIM_Cubic, tangentMode=RCTM_User, time=0.5, value=1}
            {tangentMode=RCTM_User, time=1, value=0}
           ]}}}
        Scale Curve=0.25
       ))
     )
     "Float (Alpha)"=1
    )
    Erosion Amount=0.25
    Erosion Pan Speed U=1
    Macro Distortion Amount=1
    Macro Pan Offset=1
    Macro Pan Speed=1
    Micro Distortion Amount=1
    Opacity Erosion:NiagaraFloat=&FloatFromCurve(FloatCurve=DataInterface() {propertyValues {Curve {keys [<at 0 0>,<at 1 0.5>]}}})
    Opacity Pan U Offset=1
    Opacity Pan U Speed=0.25
    Opacity Scale U=0.5
    Scale Alpha:NiagaraFloat=&FloatFromCurve(FloatCurve=DataInterface() {propertyValues {Curve {keys [
         {interpMode=RCIM_Cubic, tangentMode=RCTM_User, time=0, value=0}
         {arriveTangent=-2.8813, interpMode=RCIM_Cubic, leaveTangent=-2.8813, tangentMode=RCTM_User, time=0.1, value=1}
         {interpMode=RCIM_Cubic, tangentMode=RCTM_User, time=0.4, value=0.1}
         {interpMode=RCIM_Cubic, tangentMode=RCTM_User, time=1, value=0}
        ]}}})
   )
   &DynamicMaterialParameters(
    Index 0 Param 1=1
    Index 0 Param 2=1
    Index 0 Param 3=1
    Module.Param0WriteEnabled=true
    Module.Param1WriteEnabled=true
    Module.Param2WriteEnabled=true
    Module.Param3WriteEnabled=true
   )
  ]
  rendererClasses [&NiagaraRibbonRendererProperties]
  renderers [{rendererClass=&NiagaraRibbonRendererProperties}]
 }
 FN_AttachedRibbon002 {
  emitterUpdateScript [
   &EmitterState()
   &FN_SpawnRibbon_v2(Spawn Mode=::NewEnumerator0, SpawnRate=50)
   &FN_Integrator() {enabled=false}
  ]
  particleSpawnScript [
   &FN_InitializeRibbon_v4(
    Color:LinearColor=<rgba 1 1 1 0>
    Enable Skeletal Position=true
    Lifetime=1
    Ribbon Width=50
    SkeletalMesh:NiagaraDataInterfaceSkeletalMesh=$User.SkeletalMesh
   )
   &FN_FlipRibbonV() {enabled=false}
   &AddVelocity()
   &CameraOffset()
   &Color()
  ]
  particleUpdateScript [
   &CurlNoiseForce(Noise Frequency=150, Noise Strength=0, Pan Noise=false)
   &SolveForcesAndVelocity(Clamp Velocity=false, Limit Acceleration=false)
   &FN_UpdateRibbon_v4(Enable Skeletal Position=true, SkeletalMesh:NiagaraDataInterfaceSkeletalMesh=$User.SkeletalMesh)
   &FN_RibbonMaterial(
    Color=&MakeLinearColorFromVectorAndFloat(
     "Vector (RGB)"=&Multiply_Vector(
      A=&MakeVectorFromLinearColorRGB(Color=&ColorFromCurve(ColorCurve=DataInterface() {propertyValues {
          AlphaCurve {keys [<at 0 1 RCIM_Cubic>,<at 0.9977 1 RCIM_Cubic>]}
          BlueCurve {keys [<at 0.1175 0.4118 RCIM_Cubic>,<at 0.2412 0.7758>,<at 0.4228 0.2812>,<at 0.606 0 RCIM_Cubic>]}
          GreenCurve {keys [<at 0.1175 1 RCIM_Cubic>,<at 0.2412 0.597>,<at 0.4228 0.3027>,<at 0.606 0.3352 RCIM_Cubic>]}
          RedCurve {keys [<at 0.1175 0 RCIM_Cubic>,<at 0.2412 0.5293>,<at 0.4228 2>,<at 0.606 0.4688 RCIM_Cubic>]}
         }}))
      B=&VectorFromFloat(Value=&FloatFromCurve(FloatCurve=DataInterface() {propertyValues {Curve {keys [
            {interpMode=RCIM_Cubic, tangentMode=RCTM_User, time=0, value=0}
            {interpMode=RCIM_Cubic, tangentMode=RCTM_User, time=0.1, value=1}
            {tangentMode=RCTM_User, time=1, value=0}
           ]}}}))
     )
     "Float (Alpha)"=1
    )
    Erosion Amount=0.25
    Erosion Pan Speed U=1
    Macro Distortion Amount=5
    Macro Pan Offset=1
    Macro Pan Speed=1
    Micro Distortion Amount=1
    Opacity Erosion:NiagaraFloat=&FloatFromCurve(FloatCurve=DataInterface() {propertyValues {Curve {keys [<at 0 0>,<at 1 1>]}}})
    Opacity Pan U Offset=1
    Opacity Pan U Speed=0.25
    Opacity Scale U=0.5
    Scale Alpha:NiagaraFloat=&FloatFromCurve(FloatCurve=DataInterface() {propertyValues {Curve {keys [
         {interpMode=RCIM_Cubic, tangentMode=RCTM_User, time=0, value=0}
         {arriveTangent=-2.8813, interpMode=RCIM_Cubic, leaveTangent=-2.8813, tangentMode=RCTM_User, time=0.1, value=1}
         {interpMode=RCIM_Cubic, tangentMode=RCTM_User, time=0.4, value=0.2}
         {interpMode=RCIM_Cubic, tangentMode=RCTM_User, time=1, value=0}
        ]}}})
   ) {enabled=false}
   &DynamicMaterialParameters(
    Index 0 Param 1=1
    Index 0 Param 2=1
    Index 0 Param 3=1
    Module.Param0WriteEnabled=true
    Module.Param1WriteEnabled=true
    Module.Param2WriteEnabled=true
    Module.Param3WriteEnabled=true
   )
   &ScaleColor(
    Scale Alpha=&FloatFromCurve(
     FloatCurve=DataInterface() {propertyValues {Curve {keys [
         {interpMode=RCIM_Cubic, tangentMode=RCTM_User, time=0, value=0}
         {interpMode=RCIM_Cubic, tangentMode=RCTM_User, time=0.15, value=5}
         {interpMode=RCIM_Cubic, tangentMode=RCTM_User, time=0.3, value=1}
         {interpMode=RCIM_Cubic, tangentMode=RCTM_User, time=1, value=0}
        ]}}}
     Scale Curve=0.15
    )
    Scale RGB=&Multiply_Vector(
     A=&MakeVectorFromLinearColorRGB(Color=&ColorFromCurve(ColorCurve=DataInterface() {propertyValues {
         AlphaCurve {keys [<at 0 1 RCIM_Cubic>,<at 0.9977 1 RCIM_Cubic>]}
         BlueCurve {keys [<at 0.1953 0.5186 RCIM_Cubic>,<at 0.5951 0.7084>,<at 0.7691 1.5>,<at 0.9071 0 RCIM_Cubic>]}
         GreenCurve {keys [<at 0.1953 0 RCIM_Cubic>,<at 0.5951 0>,<at 0.7691 0.4532>,<at 0.9071 0.807 RCIM_Cubic>]}
         RedCurve {keys [<at 0.1953 1 RCIM_Cubic>,<at 0.5951 0.6272>,<at 0.7691 0>,<at 0.9071 0.1578 RCIM_Cubic>]}
        }}))
     B=&VectorFromFloat(Value=&FloatFromCurve(
       FloatCurve=DataInterface() {propertyValues {Curve {keys [
           {interpMode=RCIM_Cubic, tangentMode=RCTM_User, time=0, value=0}
           {interpMode=RCIM_Cubic, tangentMode=RCTM_User, time=0.2, value=1}
           {tangentMode=RCTM_User, time=1, value=0}
          ]}}}
       Scale Curve=2
      ))
    )
    ScaleA=true
    ScaleRGB=true
   )
  ]
  rendererClasses [&NiagaraSpriteRendererProperties, &NiagaraRibbonRendererProperties]
  renderers [
   {rendererClass=&NiagaraSpriteRendererProperties}
   {rendererClass=&NiagaraRibbonRendererProperties, rendererIndex=1}
  ]
 }
}
summary {
 systemName=NS_Biome_GC_Loop
 emitters [
  {emitterName=FN_AttachedRibbon001, rendererClasses [&NiagaraRibbonRendererProperties]}
  {
   emitterName=FN_AttachedRibbon002
   rendererClasses [&NiagaraSpriteRendererProperties, &NiagaraRibbonRendererProperties]
  }
 ]
 userVariables [
  User.bIsFrontEnd {type=NiagaraBool}
  User.bIsFrontEndPreview {type=NiagaraBool}
  User.FEVec {type=Vector3f, description="", defaultValue {struct=Vector3f, value=<0 0 0>}}
  User.RibbonIDOffset {type=NiagaraInt32}
  User.SkeletalMesh {
   type=NiagaraDataInterfaceSkeletalMesh
   defaultValue {
    struct=NiagaraExt_VariableValue_DataInterface
    value {dataInterface=&self:NiagaraDataInterfaceSkeletalMesh_0}
   }
  }
 ]
}
