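# Niagara trail effect for a biome projectile, combining a ribbon, mesh debris and a glow.
# The ribbon rises with curl noise and curve-driven color, while two random mesh variants spawn on a box surface with waveform-scaled velocity.
# A looping single sprite with a light renderer supplies the cyan glow, brightening with system age.
# System FN_Timer inversely scales spawn rates, mesh and glow size, and ribbon alpha, winding the trail down.
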NiagaraSystem /MotherGameplay/FX/Biome/NS_Biome_Projectile_Trail
dependencies {
 usedDataInterfaces [&NiagaraDataInterfaceCurve, &NiagaraDataInterfaceColorCurve]
 usedDynamicInputs [&UniformRangedFloat]
 usedModules [&SystemState, &FN_Timer]
}
emitters {
 FN_Mesh {
  emitterUpdateScript [
   &EmitterState(Life Cycle Mode=::NewEnumerator1, UseLoopDelay=false)
   &SpawnRate(
    SpawnRate=&Multiply_Float(A=30, B=&OneMinusFloat(Float=$System.FN_Timer.Value))
    Use Spawn Probability=false
   )
  ]
  particleSpawnScript [
   /SetVariables(Particles.MeshIndex=&RandomRangeInt(Minimum=0, Maximum=1)) {isSetParameters=true}
   &InitializeParticle(
    Lifetime=3
    Mesh Scale Mode=::NewEnumerator3
    Mesh Uniform Scale=&FloatFromCurve(FloatCurve=DataInterface() {propertyValues {Curve {keys [<at 0 1>,<at 1 5>]}}}, CurveIndex=$System.NormalizedLoopAge)
    UsePositionOffset=false
    Write Material Random=true
   ) {module=InitializeParticle001}
   &BoxLocation(
    Box Size=&Multiply_VectorByFloat(
     Vector=<15 15 15>
     Float=&FloatFromCurve(FloatCurve=DataInterface() {propertyValues {Curve {keys [<at 0 1>,<at 1 1>]}}}, CurveIndex=$System.NormalizedLoopAge)
    )
    MaskSpawnGroup=false
    Module.Override Seed=false
    Rotate Box=false
    Surface Only=true
   )
   &AddVelocity(Velocity=&Multiply_Vector(
     A=&Lerp_Vector(A=<5 5 5>, B=<-5 -5 -5>, Alpha=&RandomRangeFloat(Minimum=0, Maximum=1))
     B=$Output.BoxLocation.BoxNormal
    ))
   &InitialMeshOrientation(
    Orientation Vector=&MakeVector(X=0, Y=&UniformAOrBFloat(A=0), Z=0)
    Rotation=&MakeVector(X=&UniformAOrBFloat(A=-0.25, B=0.25), Y=0, Z=0)
    Use Orientation Vector=true
    Use Rotation Vector=true
   )
  ]
  particleUpdateScript [
   &ScaleVelocity(Velocity Scale=&Multiply_VectorByFloat(
     Vector=&Lerp_Vector(
      A=<-1 -1 -1>
      B=<1 1 1>
      Alpha=&Waveform(
       "Waveform [1]"=::NewEnumerator0
       "[1] Phase Scale"=&RandomRangeFloat(Minimum=0.5, Maximum=2)
       "[1] Frequency"=0.5
       "[1] Amplitude Min/Max"=<0 1>
       "[2] Phase"=$Emitter.Age
       "[3] Phase"=$Emitter.Age
      )
     )
     Float=0.5
    ))
   &CurlNoiseForce(Noise Frequency=1, Noise Strength=2500, Pan Noise=false) {enabled=false}
   &SolveForcesAndVelocity(Clamp Velocity=true, Limit Acceleration=false, Speed Limit=250)
   &ParticleState()
   &ScaleMeshSize(Scale Factor=&VectorFromFloat(Value=&OneMinusFloat(Float=$System.FN_Timer.Value)))
  ]
  rendererClasses [&NiagaraMeshRendererProperties]
  renderers [{rendererClass=&NiagaraMeshRendererProperties, rendererIndex=0}]
 }
 FN_Projectile_Light001 {
  emitterUpdateScript [
   &EmitterState(
    Inactive Response=::NewEnumerator1
    Life Cycle Mode=::NewEnumerator1
    Loop Behavior=::NewEnumerator1
    Loop Duration Mode=::NewEnumerator1
    UseLoopDelay=false
   )
   &SpawnBurst_Instantaneous(Spawn Count=1, Use Spawn Probability=false)
  ]
  particleSpawnScript [
   &InitializeParticle(Lifetime=2, Sprite Size Mode=::NewEnumerator3, Uniform Sprite Size=40, UsePositionOffset=false, Write Material Random=true) {module=InitializeParticle001}
   /Niagara/Modules/Spawn/Orientation/V2/InitialMeshOrientation(
    Orientation Axis:Vector3f=<0 0 1>
    Orientation Coordinate Space:ENiagaraCoordinateSpace=::Local
    Rotation=<1 0 0>
    Use Orientation Vector=true
    Use Rotation Vector=false
   ) {enabled=false}
   &CameraOffset(Camera Offset Amount=25)
   /SetVariables(Particles.NewLinearColor:LinearColor=<rgba 1 14.2309 25 1>) {isSetParameters=true}
  ]
  particleUpdateScript [
   &ParticleState(
    Kill Particles When Lifetime Has Elapsed=false
    Let Infinitely Lived Particles Die When Emitter Deactivates=true
    Loop Particles Lifetime=true
   )
   &InheritVelocity(Module.Limit Speed=false, Velocity Scale=<0.01 0.01 0.01>) {enabled=false}
   &SolveForcesAndVelocity(Clamp Velocity=false, Limit Acceleration=false)
   &ScaleColor(
    Scale RGB=&Multiply_VectorByFloat(
     Vector=<1 5 5>
     Float=&Multiply_Float(
      A=&FloatFromCurve(
       FloatCurve=DataInterface() {propertyValues {Curve {keys [
           {interpMode=RCIM_Cubic, tangentMode=RCTM_User, time=0, value=0.1}
           {interpMode=RCIM_Cubic, tangentMode=RCTM_User, time=0.5, value=1}
           {tangentMode=RCTM_User, time=1, value=1}
          ]}}}
       CurveIndex=$System.Age
      )
      B=&FloatFromCurve(FloatCurve=DataInterface() {propertyValues {Curve {keys [
           {interpMode=RCIM_Cubic, tangentMode=RCTM_User, time=0, value=0.7}
           {interpMode=RCIM_Cubic, tangentMode=RCTM_User, time=0.5, value=1}
           {interpMode=RCIM_Cubic, tangentMode=RCTM_User, time=1, value=0.7}
          ]}}})
     )
    )
    ScaleA=true
    ScaleRGB=true
   )
   &DynamicMaterialParameters(
    Index 0 Param 1=&RandomRangeFloat(Minimum=0.9, Maximum=0.9)
    Index 0 Param 2=&RandomRangeFloat(Minimum=-1, Maximum=1)
    Index 0 Param 3=&RandomRangeFloat(Minimum=1, Maximum=1)
    Index 0 Param 4=1
    Index 1 Param 1=&FloatFromCurve(FloatCurve=DataInterface() {propertyValues {Curve {keys [<at 0 2>,<at 1 4>]}}}, CurveIndex=$System.Age)
    Index 1 Param 2=1
    Index 1 Param 3=&FloatFromCurve(FloatCurve=DataInterface() {propertyValues {Curve {keys [<at 0 0>,<at 1 25>]}}}, CurveIndex=$System.Age)
    Module.Param0WriteEnabled=true
    Module.Param0WriteEnabled1=true
    Module.Param1WriteEnabled=true
    Module.Param1WriteEnabled1=true
    Module.Param2WriteEnabled=true
    Module.Param2WriteEnabled1=true
    Module.Param3WriteEnabled=true
    Module.Param3WriteEnabled1=true
    Write Parameter Index 1=true
   )
   /SetVariables(Particles.NewLinearColor001:LinearColor=&MultiplyLinearColor(
     Linear Color=$Particles.NewLinearColor
     Scale Factor=&MakeVector4_FromFloat(Float=&Multiply_Float(
       A=&FloatFromCurve(
        FloatCurve=DataInterface() {propertyValues {Curve {keys [
            {interpMode=RCIM_Cubic, tangentMode=RCTM_User, time=0, value=0.1}
            {interpMode=RCIM_Cubic, tangentMode=RCTM_User, time=0.5, value=1}
            {tangentMode=RCTM_User, time=1, value=1}
           ]}}}
        CurveIndex=$System.Age
       )
       B=&FloatFromCurve(
        FloatCurve=DataInterface() {propertyValues {Curve {keys [
            {interpMode=RCIM_Cubic, tangentMode=RCTM_User, time=0, value=0.7}
            {interpMode=RCIM_Cubic, tangentMode=RCTM_User, time=0.5, value=1}
            {interpMode=RCIM_Cubic, tangentMode=RCTM_User, time=1, value=0.7}
           ]}}}
        Scale Curve=&OneMinusFloat(Float=$System.FN_Timer.Value)
       )
      ))
    )) {isSetParameters=true}
   &CameraOffset(Camera Offset Amount=&FloatFromCurve(FloatCurve=DataInterface() {propertyValues {Curve {keys [<at 0.5 0>,<at 2 1>]}}}, CurveIndex=$System.Age, Scale Curve=25)) {module=CameraOffset001}
   &ScaleSpriteSize(Scale Factor=&Vector2DFromFloat(Value=&FloatFromCurve(
      FloatCurve=DataInterface() {propertyValues {Curve {keys [{tangentMode=RCTM_User, time=0, value=0}, {tangentMode=RCTM_User, time=1, value=2}]}}}
      Scale Curve=&Multiply_Float(A=2, B=&OneMinusFloat(Float=$System.FN_Timer.Value))
     )))
  ]
  rendererClasses [&NiagaraSpriteRendererProperties, &NiagaraLightRendererProperties]
  renderers [
   {rendererClass=&NiagaraSpriteRendererProperties}
   {rendererClass=&NiagaraLightRendererProperties, rendererIndex=1}
  ]
 }
 FN_Projectile_Ribbon_Fire {
  emitterUpdateScript [
   &FN_SpawnRibbon_v2(Spawn Mode=::NewEnumerator0, SpawnRate=30) {enabled=false}
   &EmitterState()
   &SpawnRate(
    SpawnRate=&Multiply_Float(A=30, B=&OneMinusFloat(Float=$System.FN_Timer.Value))
    Use Spawn Probability=false
   )
  ]
  particleSpawnScript [
   &InitializeRibbon(
    Module.Write Color=true
    Module.Write Lifetime=true
    Module.Write Mass=true
    Module.Write Material Random=true
    Module.Write Position=false
    Module.Write RibbonTwist=false
    Module.Write RibbonWidth=true
    Ribbon Width=80
   )
   &FN_InitializeRibbon_v3(Ribbon Width=30) {enabled=false}
  ]
  particleUpdateScript [
   &ParticleState()
   &ScaleSpriteSize(Scale Factor=&Vector2DFromFloat(Value=&FloatFromCurve(FloatCurve=DataInterface() {propertyValues {Curve {keys [
          {interpMode=RCIM_Cubic, tangentMode=RCTM_User, time=0.0002, value=0}
          {interpMode=RCIM_Cubic, tangentMode=RCTM_User, time=0.2188, value=0.5881}
          {interpMode=RCIM_Cubic, tangentMode=RCTM_User, time=0.9878, value=0.35}
         ]}}})))
   &GravityForce(Gravity=<0 0 1000>)
   /Niagara/Modules/Update/Forces/CurlNoiseForce(Noise Frequency=0.03, Noise Strength=250, Pan Noise=false)
   &SolveForcesAndVelocity(Clamp Velocity=false, Limit Acceleration=false)
   /Game/Effects/Niagara/Modules/Contrails/FN_UpdateRibbon/FN_UpdateRibbon_v2(Initial Color:LinearColor=$Particles.Initial.Color, Initial RibbonWidth:NiagaraFloat=$Particles.Initial.RibbonWidth) {enabled=false}
   &DynamicMaterialParameters(
    Index 0 Param 1=&FloatFromCurve(FloatCurve=DataInterface() {propertyValues {Curve {keys [<at 0 0>,<at 0.9928 1>]}}})
    Module.Param0WriteEnabled=true
    Module.Param1WriteEnabled=false
    Module.Param2WriteEnabled=false
    Module.Param3WriteEnabled=false
   ) {module=DynamicMaterialParameters001}
   &ScaleColor(
    Scale Alpha=&FloatFromCurve(
     FloatCurve=DataInterface() {propertyValues {Curve {keys [
         {arriveTangent=-0.0667, interpMode=RCIM_Cubic, leaveTangent=-0.0667, tangentMode=RCTM_User, time=0, value=0}
         <at 0.1 0>
         {arriveTangent=0.9995, interpMode=RCIM_Cubic, leaveTangent=0.9995, tangentMode=RCTM_User, time=0.4, value=1}
         {arriveTangent=-0.0405, interpMode=RCIM_Cubic, leaveTangent=-0.0405, tangentMode=RCTM_User, time=1, value=0.05}
        ]}}}
     Scale Curve=&OneMinusFloat(Float=$System.FN_Timer.Value)
    )
    Scale RGB=&MakeVectorFromLinearColorRGB(Color=&MultiplyLinearColor(
      Linear Color=&ColorFromCurve(ColorCurve=DataInterface() {propertyValues {
         AlphaCurve {keys [<at 0 0 RCIM_Cubic>,<at 0.0942 1>,<at 1 0 RCIM_Cubic>]}
         BlueCurve {keys [<at 0 0 RCIM_Cubic>,<at 0.0957 1>,<at 0.437 1>,<at 0.8556 1>,<at 1 1 RCIM_Cubic>]}
         GreenCurve {keys [<at 0 0 RCIM_Cubic>,<at 0.0957 0>,<at 0.437 0.4445>,<at 0.8556 0>,<at 1 1 RCIM_Cubic>]}
         RedCurve {keys [<at 0 0 RCIM_Cubic>,<at 0.0957 0.9462>,<at 0.437 0>,<at 0.8556 0.3423>,<at 1 1 RCIM_Cubic>]}
        }})
      Scale Factor=&MakeVector4_FromFloat(Float=&FloatFromCurve(FloatCurve=DataInterface() {propertyValues {Curve {keys [
            {arriveTangent=4.1785, interpMode=RCIM_Cubic, leaveTangent=4.1785, tangentMode=RCTM_User, time=0, value=0}
            {arriveTangent=0.9995, interpMode=RCIM_Cubic, leaveTangent=0.9995, tangentMode=RCTM_User, time=0.4, value=1}
            {arriveTangent=-0.0405, interpMode=RCIM_Cubic, leaveTangent=-0.0405, tangentMode=RCTM_User, time=1, value=0.05}
           ]}}}))
     ))
    ScaleA=true
    ScaleRGB=true
   )
  ]
  rendererClasses [&NiagaraRibbonRendererProperties]
  renderers [{rendererClass=&NiagaraRibbonRendererProperties}]
 }
}
summary {
 systemName=NS_Biome_Projectile_Trail
 emitters [
  {emitterName=FN_Projectile_Ribbon_Fire, rendererClasses [&NiagaraRibbonRendererProperties]}
  {emitterName=FN_Mesh, rendererClasses [&NiagaraMeshRendererProperties]}
  {
   emitterName=FN_Projectile_Light001
   rendererClasses [&NiagaraSpriteRendererProperties, &NiagaraLightRendererProperties]
  }
 ]
 userVariables [
  User.OnDeath {type=NiagaraBool}
  User.RibbonIDOffset {type=NiagaraInt32}
 ]
}
