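# Niagara system that fills a User.BoxSize biome volume with GPU spark sprites, killing any that leave the box.
# One emitter fires an outward burst, while three spawn continuously, two of them snapping particles onto the depth buffer.
# All sample GBuffer color at spawn, shift color over life, and scale spawn count by system distance variables.
# Two continuous emitters ramp point attraction at system age 28–30 seconds, killing particles near the attractor.
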NiagaraSystem /MotherGameplay/FX/Biome/NS_Biome_Volume
dependencies {
 usedDataInterfaces [
  &NiagaraDataInterfaceGBuffer
  &NiagaraDataInterfaceColorCurve
  &NiagaraDataInterfaceCurve
  &NiagaraDataInterfaceCamera
 ]
 usedDynamicInputs [/Game/Effects/Niagara/DynamicInputs/Math/FN_Fit]
 usedModules [
  &self:SystemSpawnScript.SystemScriptSource.SystemScriptGraph.NiagaraNodeAssignment_0.SetVariables_#1
  &SystemState
  &self:SystemSpawnScript.SystemScriptSource.SystemScriptGraph.NiagaraNodeAssignment_1.SetVariables_#2
 ]
}
emitters {
 Pop_Sparks {
  simTarget=GPUComputeSim
  emitterUpdateScript [
   &EmitterState(Life Cycle Mode=::NewEnumerator1, Loop Behavior=::NewEnumerator1, Loop Duration=1, UseLoopDelay=false)
   &SpawnBurst_Instantaneous(
    Spawn Count=&MakeIntFromFloat(Float=&Multiply_Float(A=8000, B=$System.ScaleDistanceUpdate_SpawnCount))
    Spawn Time=0.25
    Use Spawn Probability=false
   )
  ]
  particleSpawnScript [
   /Niagara/Modules/Spawn/Initialization/InitializeParticle(
    Color=&LerpLinearColors(StartColor=<rgba 1 1 1 1>, EndColor=<rgba 1 1 1 1>, LerpFactor=1)
    Lifetime=&Multiply_Float(A=&UniformRangedFloat(Minimum=1, Maximum=2), B=1)
    Module.Write Color=true
    Module.Write Lifetime=true
    Module.Write Mass=true
    Module.Write Material Random=true
    Module.Write Position=false
    Module.Write Scale=false
    Module.Write SpriteRotation=false
    Module.Write SpriteSize=true
    Position=&SimulationPosition()
    Sprite Size=&UniformRangedVector2D(
     Minimum=&Vector2DFromFloat(Value=25)
     Maximum=&Vector2DFromFloat(Value=50)
    )
   ) {enabled=false}
   /Niagara/Modules/Spawn/Initialization/V2/InitializeParticle(
    AdjustAlpha=false
    AdjustHue=true
    AdjustSaturation=false
    AdjustValue=false
    Color=<rgba 0 0 0 1>
    Color Mode=::NewEnumerator3
    Hue Shift Range:Vector2f=<0 0>
    Lifetime Max=3
    Lifetime Min=2
    Lifetime Mode=::NewEnumerator1
    Sprite Size Mode=::NewEnumerator1
    Uniform Sprite Size Max=&Multiply_Float(A=10, B=$System.ScaleDistanceUpdate)
    Uniform Sprite Size Min=&Multiply_Float(A=5, B=$System.ScaleDistanceUpdate)
    UsePositionOffset=false
    Write Material Random=true
   ) {module=InitializeParticle001}
   &SystemLocation(MaskSpawnGroup=false, Offset=$System.SpawnOffset) {enabled=false}
   &SphereLocation(
    MaskSpawnGroup=false
    Override Seed=false
    Sphere Radius=0
    Surface Only=false
    UseNonUniformScale=true
    UseOffset=false
    UseSphereOrientation=false
   )
   &BoxLocation(
    Box Offset=&MakeVector(
     X=0
     Y=0
     Z=&Multiply_Float(A=&MakeFloatFromVector(Vector=$System.BoxSize, Channel=::NewEnumerator2), B=0.5)
    )
    Box Origin:NiagaraPosition=/Niagara/DynamicInputs/Helpers/SimulationPosition()
    Box Size=$System.BoxSize
    Coordinate Space:ENiagaraCoordinateSpace=::World
    MaskSpawnGroup=false
    Module.Override Seed=false
    Rotate Box=false
    Surface Only=false
   )
   /SetVariables(Particles.InitialPos:NiagaraPosition=$Particles.Position) {isSetParameters=true, module=SetVariables_#3}
   /Game/Effects/Niagara/Modules/Spawn/FN_Spawn_GBuffer_Color()
   &ApplyInitialForces()
   &CameraOffset(Camera Offset Amount=25)
   &AddVelocityFromPoint(Module.Use Falloff Distance=true, Velocity Falloff Distance=2500, Velocity Strength=1500)
  ]
  particleUpdateScript [
   &ParticleState()
   /SetVariables(Particles.GradientColor:Vector3f=&MakeVectorFromLinearColorRGB(Color=&MakeLinearColorFromVectorAndFloat(
      "Vector (RGB)"=&MakeVectorFromLinearColorRGB(Color=&ColorFromCurve(ColorCurve=DataInterface() {propertyValues {
          AlphaCurve {keys [<at -0.0023 1 RCIM_Cubic>,<at 0.9977 1 RCIM_Cubic>]}
          BlueCurve {keys [<at 0.1175 0.3005 RCIM_Cubic>,<at 0.2682 0.2705>,<at 0.4507 0.2122>,<at 0.5951 0.5271>,<at 0.7691 1.5>,<at 0.9071 0.521 RCIM_Cubic>]}
          GreenCurve {keys [<at 0.1175 0.1878 RCIM_Cubic>,<at 0.2682 0.2623>,<at 0.4507 0.2346>,<at 0.5951 0.4287>,<at 0.7691 0.4532>,<at 0.9071 0.4851 RCIM_Cubic>]}
          RedCurve {keys [<at 0.1175 0.3372 RCIM_Cubic>,<at 0.2682 0.7305>,<at 0.4507 0.7758>,<at 0.5951 0.7084>,<at 0.7691 0>,<at 0.9071 0.807 RCIM_Cubic>]}
         }}))
      "Float (Alpha)"=1
     ))) {enabled=false, isSetParameters=true, module=SetVariables_#4}
   &KillParticlesInVolume(
    Box Size:Vector3f=$User.BoxSize
    Invert Volume=true
    Kill Shape=::NewEnumerator1
    Origin Offset:Vector3f=&MakeVector(
     X=0
     Y=0
     Z=&Multiply_Float(A=&MakeFloatFromVector(Vector=$System.BoxSize, Channel=::NewEnumerator2), B=0.5)
    )
   ) {module=KillParticlesInVolume001}
   &Drag(
    Drag=&Multiply_Float(
     A=&UniformRangedFloat(Minimum=5, Maximum=5)
     B=&FloatFromCurve(FloatCurve=DataInterface() {propertyValues {Curve {keys [
          {interpMode=RCIM_Cubic, tangentMode=RCTM_User, time=0, value=0}
          {interpMode=RCIM_Cubic, tangentMode=RCTM_User, time=0.1, value=1}
          <at 0.4 0>
          <at 1 0>
         ]}}})
    )
    Use Linear Drag=true
    Use Rotational Drag=false
   )
   &GravityForce(Gravity=&Multiply_VectorByFloat(
     Vector=&RandomRangeVector(Minimum=<200 200 -200>, Maximum=<-200 -200 200>)
     Float=&FloatFromCurve(FloatCurve=DataInterface() {propertyValues {Curve {keys [<at 0 0>,<at 1 1>]}}})
    )) {enabled=false}
   /Niagara/Modules/Update/Forces/CurlNoiseForce(
    Noise Frequency=250
    Noise Quality / Cost=::NewEnumerator0
    Noise Strength=&FloatFromCurve(
     FloatCurve=DataInterface() {propertyValues {Curve {keys [
         {interpMode=RCIM_Cubic, tangentMode=RCTM_User, time=0, value=0}
         {interpMode=RCIM_Cubic, tangentMode=RCTM_User, time=0.1, value=0.75}
         {interpMode=RCIM_Cubic, tangentMode=RCTM_User, time=0.3, value=0.75}
         {interpMode=RCIM_Cubic, tangentMode=RCTM_User, time=0.4, value=0.25}
         <at 0.6 0.75>
         {interpMode=RCIM_Cubic, tangentMode=RCTM_User, time=1, value=0}
        ]}}}
     Scale Curve=&RandomRangeFloat(Minimum=50, Maximum=250)
    )
    Pan Noise=true
    Pan Noise Field=&RandomRangeVector(Minimum=<0 -150 -20>, Maximum=<1 150 10>)
   ) {enabled=false}
   &PointAttractionForce(
    Attraction Radius=1500
    Attraction Strength=0
    Attractor Position:Vector3f=$Emitter.CameraPosition
    Attractor Position Offset Coordinate Space:ENiagaraCoordinateSpace=::Local
    Falloff Exponent=1
    Kill Within Radius=false
    Module.Use Falloff=false
    UseAttractorOffset=true
   ) {enabled=false, module=PointAttractionForce001}
   &SolveForcesAndVelocity(Clamp Velocity=false, Limit Acceleration=false)
   &ScaleSpriteSize(Scale Factor=&Vector2DFromFloat(Value=&FloatFromCurve(FloatCurve=DataInterface() {propertyValues {Curve {keys [
          {arriveTangent=5.6168, interpMode=RCIM_Cubic, leaveTangent=5.6168, tangentMode=RCTM_User, time=0, value=0}
          {interpMode=RCIM_Cubic, tangentMode=RCTM_User, time=0.1, value=0.85}
          <at 0.2 0.3>
         ]}}})))
   &ScaleColor(
    Scale Alpha=&FloatFromCurve(
     FloatCurve=DataInterface() {propertyValues {Curve {keys [
         {interpMode=RCIM_Cubic, tangentMode=RCTM_User, time=0, value=0}
         {interpMode=RCIM_Cubic, tangentMode=RCTM_User, time=0.15, value=1}
         {interpMode=RCIM_Cubic, tangentMode=RCTM_User, time=0.8, value=1}
         {interpMode=RCIM_Cubic, tangentMode=RCTM_User, time=1, value=0}
        ]}}}
     Scale Curve=0.25
    )
    Scale RGB=&Multiply_VectorByFloat(
     Vector=&Lerp_Vector(
      A=<5 5 5>
      B=&MakeVectorFromLinearColorRGB(Color=&MakeLinearColorFromVectorAndFloat(
        "Vector (RGB)"=&MakeVectorFromLinearColorRGB(Color=&ColorFromCurve(ColorCurve=DataInterface() {propertyValues {
            AlphaCurve {keys [<at 0 1 RCIM_Cubic>,<at 0.9977 1 RCIM_Cubic>]}
            BlueCurve {keys [<at 0.3717 0 RCIM_Cubic>,<at 0.423 0>,<at 0.5147 4.8858>,<at 0.7691 5>,<at 0.9071 2.5 RCIM_Cubic>]}
            GreenCurve {keys [<at 0.3717 12.397 RCIM_Cubic>,<at 0.423 2.2877>,<at 0.5147 0.8333>,<at 0.7691 1.5106>,<at 0.9071 5 RCIM_Cubic>]}
            RedCurve {keys [<at 0.3717 250 RCIM_Cubic>,<at 0.423 3>,<at 0.5147 5>,<at 0.7691 0>,<at 0.9071 2.5295 RCIM_Cubic>]}
           }}))
        "Float (Alpha)"=&FloatFromCurve(FloatCurve=DataInterface() {propertyValues {Curve {keys [
             {interpMode=RCIM_Cubic, tangentMode=RCTM_User, time=0, value=1}
             {interpMode=RCIM_Cubic, tangentMode=RCTM_User, time=0.5, value=1}
             {interpMode=RCIM_Cubic, tangentMode=RCTM_User, time=1, value=1}
            ]}}})
       ))
      Alpha=&FloatFromCurve(FloatCurve=DataInterface() {propertyValues {Curve {keys [
           {interpMode=RCIM_Cubic, tangentMode=RCTM_User, time=0, value=0}
           {interpMode=RCIM_Cubic, tangentMode=RCTM_User, time=0.1, value=0}
           {interpMode=RCIM_Cubic, tangentMode=RCTM_User, time=0.5, value=1}
          ]}}})
     )
     Float=1
    )
    ScaleA=true
    ScaleRGB=true
   )
   &DynamicMaterialParameters(Module.Param0WriteEnabled=true, Module.Param1WriteEnabled=true, Module.Param2WriteEnabled=true, Module.Param3WriteEnabled=true) {enabled=false}
   &ScaleSpriteSizeBySpeed(Max Scale Factor=<1 25>)
  ]
  rendererClasses [&NiagaraSpriteRendererProperties]
  renderers [{rendererClass=&NiagaraSpriteRendererProperties}]
 }
 Pop_Sparks001 {
  simTarget=GPUComputeSim
  emitterUpdateScript [
   &EmitterState()
   &SpawnRate(SpawnRate=&Multiply_Float(A=2500, B=$System.ScaleDistanceUpdate_SpawnCount), Use Spawn Probability=true)
  ]
  particleSpawnScript [
   /Niagara/Modules/Spawn/Initialization/InitializeParticle(
    Color=&LerpLinearColors(StartColor=<rgba 1 1 1 1>, EndColor=<rgba 1 1 1 1>, LerpFactor=1)
    Lifetime=&Multiply_Float(A=&UniformRangedFloat(Minimum=1, Maximum=2), B=1)
    Module.Write Color=true
    Module.Write Lifetime=true
    Module.Write Mass=true
    Module.Write Material Random=true
    Module.Write Position=false
    Module.Write Scale=false
    Module.Write SpriteRotation=false
    Module.Write SpriteSize=true
    Position=/Niagara/Functions/Localspace/SimulationPosition()
    Sprite Size=&UniformRangedVector2D(
     Minimum=&Vector2DFromFloat(Value=25)
     Maximum=&Vector2DFromFloat(Value=50)
    )
   ) {enabled=false}
   /SetVariables(Particles.VisibilityTag=1) {isSetParameters=true, module=SetVariables_#5}
   /Niagara/Modules/Spawn/Initialization/V2/InitializeParticle(
    AdjustAlpha=false
    AdjustHue=true
    AdjustSaturation=false
    AdjustValue=false
    Color=<rgba 0 0 0 1>
    Color Mode=::NewEnumerator3
    Hue Shift Range:Vector2f=<0 0>
    Lifetime Max=5
    Lifetime Min=3
    Lifetime Mode=::NewEnumerator1
    Sprite Size Mode=::NewEnumerator1
    UsePositionOffset=false
    Write Material Random=true
   ) {module=InitializeParticle001}
   &SystemLocation(MaskSpawnGroup=false, Offset=$System.SpawnOffset) {enabled=false}
   &SphereLocation(
    MaskSpawnGroup=false
    Override Seed=false
    Sphere Radius=250
    Surface Only=false
    UseNonUniformScale=true
    UseOffset=false
    UseSphereOrientation=false
   )
   &BoxLocation(
    Box Offset=&MakeVector(
     X=0
     Y=0
     Z=&Multiply_Float(A=&MakeFloatFromVector(Vector=$System.BoxSize, Channel=::NewEnumerator2), B=0.5)
    )
    Box Origin:NiagaraPosition=/Niagara/DynamicInputs/Helpers/SimulationPosition()
    Box Size=$System.BoxSize
    Coordinate Space:ENiagaraCoordinateSpace=::World
    MaskSpawnGroup=false
    Module.Override Seed=false
    Rotate Box=false
    Surface Only=false
   )
   /SetVariables(Particles.InitialPos:NiagaraPosition=$Particles.Position) {isSetParameters=true, module=SetVariables_#3}
   /Niagara/Modules/Math/PlaceParticlesOnDepthBuffer_GPU(Max Z Depth=15000, Module.Use Max Z Depth=true)
   &KillParticles(Kill Particles=&SetBoolByFloatComparison(
     A=&VectorLength(Vector=&Subtract_Vector(A=$Particles.InitialPos, B=Linked() {linkedVariable {name=Particles.Position, type=NiagaraPosition}}))
     B=15000
    ))
   &KillParticles(Kill Particles=&Boolean_NotOperation(Bool=$Output.PlaceParticlesOnDepthBuffer_GPU.ValidSample)) {module=KillParticles001}
   &FN_Spawn_GBuffer_Color()
   &CurlNoiseForce(Curl Noise Cone Mask Angle=50, Mask Curl Noise=true, Noise Frequency=150, Noise Strength=25, Pan Noise=true, Pan Noise Field=<0 0 -5>) {module=CurlNoiseForce001}
   &ApplyInitialForces()
   &CameraOffset(Camera Offset Amount=25)
  ]
  particleUpdateScript [
   &ParticleState()
   /SetVariables(Particles.GradientColor:Vector3f=&MakeVectorFromLinearColorRGB(Color=&MakeLinearColorFromVectorAndFloat(
      "Vector (RGB)"=&MakeVectorFromLinearColorRGB(Color=&ColorFromCurve(ColorCurve=DataInterface() {propertyValues {
          AlphaCurve {keys [<at -0.0023 1 RCIM_Cubic>,<at 0.9977 1 RCIM_Cubic>]}
          BlueCurve {keys [<at 0.1175 0.3005 RCIM_Cubic>,<at 0.2682 0.2705>,<at 0.4507 0.2122>,<at 0.5951 0.5271>,<at 0.7691 1.5>,<at 0.9071 0.521 RCIM_Cubic>]}
          GreenCurve {keys [<at 0.1175 0.1878 RCIM_Cubic>,<at 0.2682 0.2623>,<at 0.4507 0.2346>,<at 0.5951 0.4287>,<at 0.7691 0.4532>,<at 0.9071 0.4851 RCIM_Cubic>]}
          RedCurve {keys [<at 0.1175 0.3372 RCIM_Cubic>,<at 0.2682 0.7305>,<at 0.4507 0.7758>,<at 0.5951 0.7084>,<at 0.7691 0>,<at 0.9071 0.807 RCIM_Cubic>]}
         }}))
      "Float (Alpha)"=1
     ))) {enabled=false, isSetParameters=true, module=SetVariables_#4}
   /SetVariables(Particles.VisibilityTag=&MakeCustomIntFromBool(Bool=$Transient.FirstFrame, True INT=1, False INT=0)) {isSetParameters=true, module=SetVariables_#6}
   &KillParticlesInVolume(
    Box Size:Vector3f=$User.BoxSize
    Invert Volume=true
    Kill Shape=::NewEnumerator1
    Origin Offset:Vector3f=&MakeVector(
     X=0
     Y=0
     Z=&Multiply_Float(A=&MakeFloatFromVector(Vector=$System.BoxSize, Channel=::NewEnumerator2), B=0.5)
    )
   )
   &Drag(
    Drag=&Multiply_Float(
     A=&UniformRangedFloat(Minimum=0, Maximum=0)
     B=&FloatFromCurve(FloatCurve=DataInterface() {propertyValues {Curve {keys [<at 0.1 1>,<at 0.25 0.1>]}}})
    )
    Use Linear Drag=true
    Use Rotational Drag=false
   )
   &GravityForce(Gravity=&Multiply_VectorByFloat(
     Vector=<0 0 100>
     Float=&FloatFromCurve(FloatCurve=DataInterface() {propertyValues {Curve {keys [<at 0 0>,<at 1 1>]}}})
    ))
   &CurlNoiseForce(
    Curl Noise Cone Mask Angle=35
    Mask Curl Noise=true
    Noise Frequency=250
    Noise Quality / Cost=::NewEnumerator0
    Noise Strength=&FloatFromCurve(FloatCurve=DataInterface() {propertyValues {Curve {keys [<at 0 1>,<at 0.25 0>]}}}, Scale Curve=250)
    Pan Noise=true
    Pan Noise Field=<0 10 0>
   )
   &PointAttractionForce(
    Attraction Radius=1500
    Attraction Strength=&FloatFromCurve(FloatCurve=DataInterface() {propertyValues {Curve {keys [<at 28 0>,<at 30 1>]}}}, CurveIndex=$Engine.System.Age)
    Attractor Position Offset Coordinate Space:ENiagaraCoordinateSpace=::Local
    Falloff Exponent=1
    Kill Radius=250
    Kill Within Radius=true
    Module.Use Falloff=false
    UseAttractorOffset=true
   ) {module=PointAttractionForce001}
   &SolveForcesAndVelocity(Clamp Velocity=false, Limit Acceleration=false)
   &ScaleSpriteSize(Scale Factor=&Vector2DFromFloat(Value=&FloatFromCurve(
      FloatCurve=DataInterface() {propertyValues {Curve {keys [
          {arriveTangent=7.1136, interpMode=RCIM_Cubic, leaveTangent=7.1136, tangentMode=RCTM_User, time=0, value=0}
          {interpMode=RCIM_Cubic, tangentMode=RCTM_User, time=0.1, value=1.375}
          {interpMode=RCIM_Cubic, tangentMode=RCTM_User, time=0.4, value=0.75}
          {interpMode=RCIM_Cubic, tangentMode=RCTM_User, time=0.7, value=1.125}
          <at 1 1.15 RCIM_Cubic>
         ]}}}
      Scale Curve=&Multiply_Float(A=1, B=$System.ScaleDistanceUpdate)
     )))
   &ScaleSpriteSizeBySpeed(Max Scale Factor=<1 5>)
   &ScaleColor(
    Scale Alpha=&FloatFromCurve(FloatCurve=DataInterface() {propertyValues {Curve {keys [
         {interpMode=RCIM_Cubic, tangentMode=RCTM_User, time=0, value=0}
         {interpMode=RCIM_Cubic, tangentMode=RCTM_User, time=0.15, value=1}
         {interpMode=RCIM_Cubic, tangentMode=RCTM_User, time=0.8, value=1}
         {interpMode=RCIM_Cubic, tangentMode=RCTM_User, time=1, value=0}
        ]}}})
    Scale RGB=&Multiply_VectorByFloat(
     Vector=&Lerp_Vector(
      A=<1 1 1>
      B=&MakeVectorFromLinearColorRGB(Color=&MakeLinearColorFromVectorAndFloat(
        "Vector (RGB)"=&MakeVectorFromLinearColorRGB(Color=&ColorFromCurve(ColorCurve=DataInterface() {propertyValues {
            AlphaCurve {keys [<at 0 1 RCIM_Cubic>,<at 0.9977 1 RCIM_Cubic>]}
            BlueCurve {keys [<at 0.1175 0.0035 RCIM_Cubic>,<at 0.2682 0.1377>,<at 0.4507 0.2122>,<at 0.5951 0.5271>,<at 0.7691 1.5>,<at 0.9071 0.521 RCIM_Cubic>]}
            GreenCurve {keys [<at 0.1175 0.0068 RCIM_Cubic>,<at 0.2682 0.6131>,<at 0.4507 0.2346>,<at 0.5951 0.4287>,<at 0.7691 0.4532>,<at 0.9071 0.4852 RCIM_Cubic>]}
            RedCurve {keys [<at 0.1175 0.3372 RCIM_Cubic>,<at 0.2682 0.7305>,<at 0.4507 0.7758>,<at 0.5951 0.7084>,<at 0.7691 0>,<at 0.9071 0.807 RCIM_Cubic>]}
           }}))
        "Float (Alpha)"=&FloatFromCurve(FloatCurve=DataInterface() {propertyValues {Curve {keys [
             {interpMode=RCIM_Cubic, tangentMode=RCTM_User, time=0, value=1}
             {interpMode=RCIM_Cubic, tangentMode=RCTM_User, time=0.5, value=1}
             {interpMode=RCIM_Cubic, tangentMode=RCTM_User, time=1, value=1}
            ]}}})
       ))
      Alpha=&FloatFromCurve(FloatCurve=DataInterface() {propertyValues {Curve {keys [
           {interpMode=RCIM_Cubic, tangentMode=RCTM_User, time=0, value=0}
           {interpMode=RCIM_Cubic, tangentMode=RCTM_User, time=0.1, value=0}
           {interpMode=RCIM_Cubic, tangentMode=RCTM_User, time=1, value=1}
          ]}}})
     )
     Float=1
    )
    ScaleA=true
    ScaleRGB=true
   )
   &DynamicMaterialParameters(Module.Param0WriteEnabled=true, Module.Param1WriteEnabled=true, Module.Param2WriteEnabled=true, Module.Param3WriteEnabled=true) {enabled=false}
  ]
  rendererClasses [&NiagaraSpriteRendererProperties]
  renderers [{rendererClass=&NiagaraSpriteRendererProperties}]
 }
 Pop_Sparks003 {
  simTarget=GPUComputeSim
  emitterUpdateScript [
   &EmitterState()
   &SpawnRate(SpawnRate=&Multiply_Float(A=1000, B=$System.ScaleDistanceUpdate_SpawnCount), Use Spawn Probability=true)
  ]
  particleSpawnScript [
   /Niagara/Modules/Spawn/Initialization/InitializeParticle(
    Color=&LerpLinearColors(StartColor=<rgba 1 1 1 1>, EndColor=<rgba 1 1 1 1>, LerpFactor=1)
    Lifetime=&Multiply_Float(A=&UniformRangedFloat(Minimum=1, Maximum=2), B=1)
    Module.Write Color=true
    Module.Write Lifetime=true
    Module.Write Mass=true
    Module.Write Material Random=true
    Module.Write Position=false
    Module.Write Scale=false
    Module.Write SpriteRotation=false
    Module.Write SpriteSize=true
    Position=/Niagara/Functions/Localspace/SimulationPosition()
    Sprite Size=&UniformRangedVector2D(
     Minimum=&Vector2DFromFloat(Value=25)
     Maximum=&Vector2DFromFloat(Value=50)
    )
   ) {enabled=false}
   /SetVariables(Particles.VisibilityTag=1) {isSetParameters=true, module=SetVariables_#5}
   /Niagara/Modules/Spawn/Initialization/V2/InitializeParticle(
    AdjustAlpha=false
    AdjustHue=true
    AdjustSaturation=false
    AdjustValue=false
    Color Mode=::NewEnumerator3
    Lifetime Max=5
    Lifetime Min=4
    Lifetime Mode=::NewEnumerator1
    Sprite Size Mode=::NewEnumerator1
    Uniform Sprite Size Max=&Multiply_Float(A=25, B=$System.ScaleDistanceUpdate)
    Uniform Sprite Size Min=&Multiply_Float(A=15, B=$System.ScaleDistanceUpdate)
    UsePositionOffset=false
    Write Material Random=true
   ) {module=InitializeParticle001}
   &SystemLocation(MaskSpawnGroup=false, Offset=$System.SpawnOffset) {enabled=false}
   &SphereLocation(
    MaskSpawnGroup=false
    Override Seed=false
    Sphere Radius=5000
    Surface Only=false
    UseNonUniformScale=true
    UseOffset=false
    UseSphereOrientation=false
   )
   &BoxLocation(Box Size=$System.BoxSize, MaskSpawnGroup=false, Module.Override Seed=false, Rotate Box=false, Surface Only=false)
   /SetVariables(Particles.InitialPos:NiagaraPosition=$Particles.Position) {isSetParameters=true, module=SetVariables_#3}
   &PlaceParticlesOnDepthBuffer_GPU(Max Z Depth=2500, Module.Use Max Z Depth=true)
   &FN_Spawn_GBuffer_Color()
   &KillParticles(Kill Particles=&SetBoolByFloatComparison(
     A=&VectorLength(Vector=&Subtract_Vector(A=$Particles.InitialPos, B=Linked() {linkedVariable {name=Particles.Position, type=NiagaraPosition}}))
     B=5000
    ))
   &KillParticles(Kill Particles=&Boolean_NotOperation(Bool=$Output.PlaceParticlesOnDepthBuffer_GPU.ValidSample)) {module=KillParticles001}
   &CurlNoiseForce(Noise Frequency=150, Noise Strength=1, Pan Noise=true, Pan Noise Field=<0 0 -5>) {module=CurlNoiseForce001}
   &ApplyInitialForces()
   &CameraOffset(Camera Offset Amount=15)
   &SubUVAnimation(End Frame=3, SubUV Animation Mode:ENiagara_SubUVLookupMode=::NewEnumerator2)
  ]
  particleUpdateScript [
   &ParticleState()
   /SetVariables() {isSetParameters=true}
   /SetVariables(Particles.GradientColor:Vector3f=&MakeVectorFromLinearColorRGB(Color=&MakeLinearColorFromVectorAndFloat(
      "Vector (RGB)"=&MakeVectorFromLinearColorRGB(Color=&ColorFromCurve(ColorCurve=DataInterface() {propertyValues {
          AlphaCurve {keys [<at -0.0023 1 RCIM_Cubic>,<at 0.9977 1 RCIM_Cubic>]}
          BlueCurve {keys [<at 0.1175 0.3005 RCIM_Cubic>,<at 0.2682 0.2705>,<at 0.4507 0.2122>,<at 0.5951 0.5271>,<at 0.7691 1.5>,<at 0.9071 0.521 RCIM_Cubic>]}
          GreenCurve {keys [<at 0.1175 0.1878 RCIM_Cubic>,<at 0.2682 0.2623>,<at 0.4507 0.2346>,<at 0.5951 0.4287>,<at 0.7691 0.4532>,<at 0.9071 0.4851 RCIM_Cubic>]}
          RedCurve {keys [<at 0.1175 0.3372 RCIM_Cubic>,<at 0.2682 0.7305>,<at 0.4507 0.7758>,<at 0.5951 0.7084>,<at 0.7691 0>,<at 0.9071 0.807 RCIM_Cubic>]}
         }}))
      "Float (Alpha)"=1
     ))) {enabled=false, isSetParameters=true, module=SetVariables_#4}
   &KillParticlesInVolume(
    Box Size:Vector3f=$User.BoxSize
    Invert Volume=true
    Kill Shape=::NewEnumerator1
    Origin Offset:Vector3f=&MakeVector(
     X=0
     Y=0
     Z=&Multiply_Float(A=&MakeFloatFromVector(Vector=$System.BoxSize, Channel=::NewEnumerator2), B=0.5)
    )
   ) {module=KillParticlesInVolume001}
   /SetVariables(Particles.VisibilityTag=&MakeCustomIntFromBool(Bool=$Transient.FirstFrame, True INT=1, False INT=0)) {isSetParameters=true, module=SetVariables_#6}
   &GravityForce(Gravity=&Multiply_VectorByFloat(
     Vector=<0 0 50>
     Float=&FloatFromCurve(FloatCurve=DataInterface() {propertyValues {Curve {keys [<at 0 0>,<at 1 1>]}}})
    ))
   &CurlNoiseForce(
    Noise Frequency=250
    Noise Quality / Cost=::NewEnumerator0
    Noise Strength=&FloatFromCurve(FloatCurve=DataInterface() {propertyValues {Curve {keys [<at 0 0>,<at 0.25 1>]}}}, Scale Curve=150)
    Pan Noise=true
    Pan Noise Field=<0 10 0>
   )
   &PointAttractionForce(
    Attraction Radius=400
    Attraction Strength=-1
    Attractor Position:Vector3f=$Emitter.CameraPosition
    Attractor Position Offset:Vector3f=<0 0 -500>
    Falloff Exponent=1
    Kill Within Radius=false
    Module.Use Falloff=true
    UseAttractorOffset=true
   ) {module=PointAttractionForce001}
   &Drag(
    Drag=&Multiply_Float(
     A=&UniformRangedFloat(Minimum=0, Maximum=0)
     B=&FloatFromCurve(FloatCurve=DataInterface() {propertyValues {Curve {keys [<at 0.1 1>,<at 0.25 0.1>]}}})
    )
    Use Linear Drag=true
    Use Rotational Drag=false
   )
   &SolveForcesAndVelocity(Clamp Velocity=false, Limit Acceleration=false)
   &ScaleSpriteSize(Scale Factor=&Vector2DFromFloat(Value=&FloatFromCurve(FloatCurve=DataInterface() {propertyValues {Curve {keys [
          {arriveTangent=7.1136, interpMode=RCIM_Cubic, leaveTangent=7.1136, tangentMode=RCTM_User, time=0, value=0}
          {interpMode=RCIM_Cubic, tangentMode=RCTM_User, time=0.1, value=1.375}
          {interpMode=RCIM_Cubic, tangentMode=RCTM_User, time=0.4, value=0.75}
          {interpMode=RCIM_Cubic, tangentMode=RCTM_User, time=0.7, value=1.125}
          <at 1 1.15 RCIM_Cubic>
         ]}}})))
   &ScaleColor(
    Scale Alpha=&FloatFromCurve(
     FloatCurve=DataInterface() {propertyValues {Curve {keys [
         {interpMode=RCIM_Cubic, tangentMode=RCTM_User, time=0, value=0}
         {interpMode=RCIM_Cubic, tangentMode=RCTM_User, time=0.15, value=1}
         {interpMode=RCIM_Cubic, tangentMode=RCTM_User, time=0.8, value=1}
         {interpMode=RCIM_Cubic, tangentMode=RCTM_User, time=1, value=0}
        ]}}}
     Scale Curve=2
    )
    Scale RGB=&Multiply_VectorByFloat(
     Vector=&Lerp_Vector(
      A=<5 5 5>
      B=&MakeVectorFromLinearColorRGB(Color=&MakeLinearColorFromVectorAndFloat(
        "Vector (RGB)"=&MakeVectorFromLinearColorRGB(Color=&ColorFromCurve(ColorCurve=DataInterface() {propertyValues {
            AlphaCurve {keys [<at 0 1 RCIM_Cubic>,<at 0.9977 1 RCIM_Cubic>]}
            BlueCurve {keys [<at 0.1175 0.5 RCIM_Cubic>,<at 0.2682 0>,<at 0.5147 4.9315>,<at 0.7691 5>,<at 0.9071 2.5 RCIM_Cubic>]}
            GreenCurve {keys [<at 0.1175 0.5248 RCIM_Cubic>,<at 0.2682 1.5251>,<at 0.5147 2.5>,<at 0.7691 1.5106>,<at 0.9071 5 RCIM_Cubic>]}
            RedCurve {keys [<at 0.1175 1 RCIM_Cubic>,<at 0.2682 2>,<at 0.5147 5>,<at 0.7691 0>,<at 0.9071 2.5295 RCIM_Cubic>]}
           }}))
        "Float (Alpha)"=&FloatFromCurve(FloatCurve=DataInterface() {propertyValues {Curve {keys [
             {interpMode=RCIM_Cubic, tangentMode=RCTM_User, time=0, value=1}
             {interpMode=RCIM_Cubic, tangentMode=RCTM_User, time=0.5, value=1}
             {interpMode=RCIM_Cubic, tangentMode=RCTM_User, time=1, value=1}
            ]}}})
       ))
      Alpha=&FloatFromCurve(FloatCurve=DataInterface() {propertyValues {Curve {keys [
           {interpMode=RCIM_Cubic, tangentMode=RCTM_User, time=0, value=0}
           {interpMode=RCIM_Cubic, tangentMode=RCTM_User, time=0.1, value=0}
           {interpMode=RCIM_Cubic, tangentMode=RCTM_User, time=0.5, value=1}
          ]}}})
     )
     Float=1
    )
    ScaleA=true
    ScaleRGB=true
   )
   &DynamicMaterialParameters(
    Index 0 Param 2=1
    Index 0 Param 3=1
    Index 0 Param 4=15
    Module.Param0WriteEnabled=true
    Module.Param1WriteEnabled=true
    Module.Param2WriteEnabled=true
    Module.Param3WriteEnabled=true
   )
  ]
  rendererClasses [&NiagaraSpriteRendererProperties]
  renderers [{rendererClass=&NiagaraSpriteRendererProperties}]
 }
 Pop_Sparks004 {
  simTarget=GPUComputeSim
  emitterUpdateScript [
   &EmitterState()
   &SpawnRate(SpawnRate=&Multiply_Float(A=1000, B=$System.ScaleDistanceUpdate_SpawnCount), Use Spawn Probability=true)
  ]
  particleSpawnScript [
   /Niagara/Modules/Spawn/Initialization/InitializeParticle(
    Color=&LerpLinearColors(StartColor=<rgba 1 1 1 1>, EndColor=<rgba 1 1 1 1>, LerpFactor=1)
    Lifetime=&Multiply_Float(A=&UniformRangedFloat(Minimum=1, Maximum=2), B=1)
    Module.Write Color=true
    Module.Write Lifetime=true
    Module.Write Mass=true
    Module.Write Material Random=true
    Module.Write Position=false
    Module.Write Scale=false
    Module.Write SpriteRotation=false
    Module.Write SpriteSize=true
    Position=&SimulationPosition()
    Sprite Size=&UniformRangedVector2D(
     Minimum=&Vector2DFromFloat(Value=25)
     Maximum=&Vector2DFromFloat(Value=50)
    )
   ) {enabled=false}
   /Niagara/Modules/Spawn/Initialization/V2/InitializeParticle(
    AdjustAlpha=false
    AdjustHue=true
    AdjustSaturation=false
    AdjustValue=false
    Color=<rgba 0 0 0 1>
    Color Mode=::NewEnumerator3
    Hue Shift Range:Vector2f=<0 0>
    Lifetime Max=4
    Lifetime Min=2
    Lifetime Mode=::NewEnumerator1
    Sprite Size Mode=::NewEnumerator1
    UsePositionOffset=false
    Write Material Random=true
   ) {module=InitializeParticle001}
   &SystemLocation(MaskSpawnGroup=false, Offset=$System.SpawnOffset) {enabled=false}
   &SphereLocation(
    MaskSpawnGroup=false
    Override Seed=false
    Sphere Radius=0
    Surface Only=false
    UseNonUniformScale=true
    UseOffset=false
    UseSphereOrientation=false
   )
   &BoxLocation(
    Box Offset=&MakeVector(
     X=0
     Y=0
     Z=&Multiply_Float(A=&MakeFloatFromVector(Vector=$System.BoxSize, Channel=::NewEnumerator2), B=0.5)
    )
    Box Origin:NiagaraPosition=/Niagara/DynamicInputs/Helpers/SimulationPosition()
    Box Size=$System.BoxSize
    Coordinate Space:ENiagaraCoordinateSpace=::World
    MaskSpawnGroup=false
    Module.Override Seed=false
    Rotate Box=false
    Surface Only=false
   )
   /SetVariables(Particles.InitialPos:NiagaraPosition=$Particles.Position) {isSetParameters=true, module=SetVariables_#3}
   &FN_Spawn_GBuffer_Color()
   &CurlNoiseForce(Curl Noise Cone Mask Angle=50, Mask Curl Noise=true, Noise Frequency=150, Noise Strength=25, Pan Noise=true, Pan Noise Field=<0 0 -5>) {module=CurlNoiseForce001}
   &ApplyInitialForces()
   &CameraOffset(Camera Offset Amount=25)
   &AddVelocity()
   &AddVelocityInCone(Cone Angle=360, Cone Axis=<0 0 1>, Module.Use Velocity Falloff On Cone Axis=true, Velocity Strength=50)
  ]
  particleUpdateScript [
   &ParticleState()
   /SetVariables(Particles.GradientColor:Vector3f=&MakeVectorFromLinearColorRGB(Color=&MakeLinearColorFromVectorAndFloat(
      "Vector (RGB)"=&MakeVectorFromLinearColorRGB(Color=&ColorFromCurve(ColorCurve=DataInterface() {propertyValues {
          AlphaCurve {keys [<at -0.0023 1 RCIM_Cubic>,<at 0.9977 1 RCIM_Cubic>]}
          BlueCurve {keys [<at 0.1175 0.3005 RCIM_Cubic>,<at 0.2682 0.2705>,<at 0.4507 0.2122>,<at 0.5951 0.5271>,<at 0.7691 1.5>,<at 0.9071 0.521 RCIM_Cubic>]}
          GreenCurve {keys [<at 0.1175 0.1878 RCIM_Cubic>,<at 0.2682 0.2623>,<at 0.4507 0.2346>,<at 0.5951 0.4287>,<at 0.7691 0.4532>,<at 0.9071 0.4851 RCIM_Cubic>]}
          RedCurve {keys [<at 0.1175 0.3372 RCIM_Cubic>,<at 0.2682 0.7305>,<at 0.4507 0.7758>,<at 0.5951 0.7084>,<at 0.7691 0>,<at 0.9071 0.807 RCIM_Cubic>]}
         }}))
      "Float (Alpha)"=1
     ))) {enabled=false, isSetParameters=true, module=SetVariables_#4}
   &KillParticlesInVolume(
    Box Size:Vector3f=$User.BoxSize
    Invert Volume=true
    Kill Shape=::NewEnumerator1
    Origin Offset:Vector3f=&MakeVector(
     X=0
     Y=0
     Z=&Multiply_Float(A=&MakeFloatFromVector(Vector=$System.BoxSize, Channel=::NewEnumerator2), B=0.5)
    )
   ) {module=KillParticlesInVolume001}
   &Drag(
    Drag=&Multiply_Float(
     A=&UniformRangedFloat(Minimum=0, Maximum=0)
     B=&FloatFromCurve(FloatCurve=DataInterface() {propertyValues {Curve {keys [<at 0.1 1>,<at 0.25 0.1>]}}})
    )
    Use Linear Drag=true
    Use Rotational Drag=false
   )
   &GravityForce(Gravity=&Multiply_VectorByFloat(
     Vector=&RandomRangeVector(Minimum=<200 200 -200>, Maximum=<-200 -200 200>)
     Float=&FloatFromCurve(FloatCurve=DataInterface() {propertyValues {Curve {keys [<at 0 0>,<at 1 1>]}}})
    ))
   &CurlNoiseForce(
    Noise Frequency=250
    Noise Quality / Cost=::NewEnumerator0
    Noise Strength=&FloatFromCurve(
     FloatCurve=DataInterface() {propertyValues {Curve {keys [
         {interpMode=RCIM_Cubic, tangentMode=RCTM_User, time=0, value=0}
         {interpMode=RCIM_Cubic, tangentMode=RCTM_User, time=0.1, value=0.75}
         {interpMode=RCIM_Cubic, tangentMode=RCTM_User, time=0.3, value=0.75}
         {interpMode=RCIM_Cubic, tangentMode=RCTM_User, time=0.4, value=0.25}
         <at 0.6 0.75>
         {interpMode=RCIM_Cubic, tangentMode=RCTM_User, time=1, value=0}
        ]}}}
     Scale Curve=&RandomRangeFloat(Minimum=50, Maximum=250)
    )
    Pan Noise=true
    Pan Noise Field=&RandomRangeVector(Minimum=<0 -150 -20>, Maximum=<1 150 10>)
   )
   &PointAttractionForce(
    Attraction Radius=3000
    Attraction Strength=&FloatFromCurve(FloatCurve=DataInterface() {propertyValues {Curve {keys [<at 28 0>,<at 30 1>]}}}, CurveIndex=$Engine.System.Age)
    Attractor Position Offset Coordinate Space:ENiagaraCoordinateSpace=::Local
    Falloff Exponent=1
    Kill Radius=250
    Kill Within Radius=true
    Module.Use Falloff=false
    UseAttractorOffset=true
   )
   &SolveForcesAndVelocity(Clamp Velocity=false, Limit Acceleration=false)
   &ScaleSpriteSize(Scale Factor=&Vector2DFromFloat(Value=&FloatFromCurve(
      FloatCurve=DataInterface() {propertyValues {Curve {keys [
          {arriveTangent=7.1136, interpMode=RCIM_Cubic, leaveTangent=7.1136, tangentMode=RCTM_User, time=0, value=0}
          {interpMode=RCIM_Cubic, tangentMode=RCTM_User, time=0.1, value=1.375}
          {interpMode=RCIM_Cubic, tangentMode=RCTM_User, time=0.4, value=0.75}
          {interpMode=RCIM_Cubic, tangentMode=RCTM_User, time=0.7, value=1.125}
          <at 1 1.15 RCIM_Cubic>
         ]}}}
      Scale Curve=&Multiply_Float(A=1, B=$System.ScaleDistanceUpdate)
     )))
   &ScaleSpriteSizeBySpeed(
    Max Scale Factor=<1 5>
    Sample Scale Factor By Curve=true
    Scale Factor Curve=DataInterface() {propertyValues {Curve {keys [<at 0.4 0>,<at 0.5 1>]}}}
   )
   &ScaleColor(
    Scale Alpha=&FloatFromCurve(
     FloatCurve=DataInterface() {propertyValues {Curve {keys [
         {interpMode=RCIM_Cubic, tangentMode=RCTM_User, time=0, value=0}
         {interpMode=RCIM_Cubic, tangentMode=RCTM_User, time=0.15, value=1}
         {interpMode=RCIM_Cubic, tangentMode=RCTM_User, time=0.8, value=1}
         {interpMode=RCIM_Cubic, tangentMode=RCTM_User, time=1, value=0}
        ]}}}
     Scale Curve=0.25
    )
    Scale RGB=&Multiply_VectorByFloat(
     Vector=&Lerp_Vector(
      A=<5 5 5>
      B=&MakeVectorFromLinearColorRGB(Color=&MakeLinearColorFromVectorAndFloat(
        "Vector (RGB)"=&MakeVectorFromLinearColorRGB(Color=&ColorFromCurve(ColorCurve=DataInterface() {propertyValues {
            AlphaCurve {keys [<at 0 1 RCIM_Cubic>,<at 0.9977 1 RCIM_Cubic>]}
            BlueCurve {keys [<at 0.3717 0 RCIM_Cubic>,<at 0.423 0>,<at 0.5147 4.8858>,<at 0.7691 5>,<at 0.9071 2.5 RCIM_Cubic>]}
            GreenCurve {keys [<at 0.3717 12.397 RCIM_Cubic>,<at 0.423 2.2877>,<at 0.5147 0.8333>,<at 0.7691 1.5106>,<at 0.9071 5 RCIM_Cubic>]}
            RedCurve {keys [<at 0.3717 250 RCIM_Cubic>,<at 0.423 3>,<at 0.5147 5>,<at 0.7691 0>,<at 0.9071 2.5295 RCIM_Cubic>]}
           }}))
        "Float (Alpha)"=&FloatFromCurve(FloatCurve=DataInterface() {propertyValues {Curve {keys [
             {interpMode=RCIM_Cubic, tangentMode=RCTM_User, time=0, value=1}
             {interpMode=RCIM_Cubic, tangentMode=RCTM_User, time=0.5, value=1}
             {interpMode=RCIM_Cubic, tangentMode=RCTM_User, time=1, value=1}
            ]}}})
       ))
      Alpha=&FloatFromCurve(FloatCurve=DataInterface() {propertyValues {Curve {keys [
           {interpMode=RCIM_Cubic, tangentMode=RCTM_User, time=0, value=0}
           {interpMode=RCIM_Cubic, tangentMode=RCTM_User, time=0.1, value=0}
           {interpMode=RCIM_Cubic, tangentMode=RCTM_User, time=0.5, value=1}
          ]}}})
     )
     Float=1
    )
    ScaleA=true
    ScaleRGB=true
   )
   &DynamicMaterialParameters(Module.Param0WriteEnabled=true, Module.Param1WriteEnabled=true, Module.Param2WriteEnabled=true, Module.Param3WriteEnabled=true) {enabled=false}
  ]
  rendererClasses [&NiagaraSpriteRendererProperties]
  renderers [{rendererClass=&NiagaraSpriteRendererProperties}]
 }
}
summary {
 systemName=NS_Biome_Volume
 emitters [
  {emitterName=Pop_Sparks001, simTarget=GPUComputeSim, rendererClasses [&NiagaraSpriteRendererProperties]}
  {emitterName=Pop_Sparks003, simTarget=GPUComputeSim, rendererClasses [&NiagaraSpriteRendererProperties]}
  {emitterName=Pop_Sparks004, simTarget=GPUComputeSim, rendererClasses [&NiagaraSpriteRendererProperties]}
  {emitterName=Pop_Sparks, simTarget=GPUComputeSim, rendererClasses [&NiagaraSpriteRendererProperties]}
 ]
 userVariables [User.BoxSize {type=Vector3f, description="", defaultValue {struct=Vector3f, value=<0 0 0>}}]
}
